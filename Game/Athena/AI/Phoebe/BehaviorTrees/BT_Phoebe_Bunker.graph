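# Behavior tree for Phoebe bot bunker behavior on BB_Phoebe: building a bunker, then holding it.
# A looping root selector prioritizes construction when the blackboard build status allows it.
# Otherwise it moves to the bunker destination, heals if near it with the target far enough away, or waits there.
# Evaluator services on that fallback handle scan/sound focus and cautious evasion, plus healing at a huge interval that effectively evaluates only on entry.
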BehaviorTree /Game/Athena/AI/Phoebe/BehaviorTrees/BT_Phoebe_Bunker
blackboard=&BB_Phoebe
root {
 name=Selector
 class=BTComposite_Selector
 children [
  Build Construction {
   class=FortAthenaBTTask_Build
   children []
   decorators [Can Build Bunker {
     class=BTDecorator_Blackboard
     properties {
      arithmeticOperation=GreaterOrEqual
      intValue=4
      nodeName=Can Build Bunker
      stringValue=EExecutionStatus::ExecutionAllowed
      blackboardKey {selectedKeyName=AIEvaluator_Bunker_BuildExecutionStatus}
     }
    }]
   properties {
    bIgnoreRestartSelf=false
    executionStatusKeyName=AIEvaluator_Bunker_BuildExecutionStatus
    focalPointName=AIEvaluator_Global_FocalPoint
    nodeName=Build Construction
   }
   services []
  }
  Selector {
   class=BTComposite_Selector
   children [
    Bot Move To {
     class=FortAthenaBTTask_BotMoveTo
     decorators [Distance to Bunker Position {
       class=FortBTDecorator_DistanceBetween
       properties {
        flowAbortMode=Self
        nodeName=Distance to Bunker Position
        operator=Greater
        blackboardKeyB {selectedKeyName=AIEvaluator_Bunker_Destination}
        specifiedDistance {defaultValue=100}
       }
      }]
     properties {
      movementResultKeyName=None
      nodeName=Bot Move To
      acceptableRadius {defaultValue=5}
      allowRandomWobble {defaultValue=true}
      bAllowStrafe {defaultValue=false}
      blackboardKey {selectedKeyName=AIEvaluator_Bunker_Destination}
      observedBlackboardValueTolerance {defaultValue=4.75}
     }
    }
    Use Heal Item {
     class=FortAthenaBTTask_UseItem
     decorators [
      Can Heal {
       class=BTDecorator_Blackboard
       properties {
        arithmeticOperation=GreaterOrEqual
        intValue=3
        nodeName=Can Heal
        stringValue=EExecutionStatus::ExecutionPending
        blackboardKey {selectedKeyName=AIEvaluator_Healing_ExecutionStatus}
       }
      }
      Distance to Bunker Position {
       class=FortBTDecorator_DistanceBetween
       properties {
        flowAbortMode=None
        nodeName=Distance to Bunker Position
        operator=Less
        blackboardKeyB {selectedKeyName=AIEvaluator_Bunker_Destination}
        specifiedDistance {defaultValue=250}
       }
      }
      Distance to Target {
       class=FortBTDecorator_DistanceBetween
       properties {
        flowAbortMode=None
        nodeName=Distance to Target
        operator=Greater
        blackboardKeyB {selectedKeyName=AIEvaluator_Global_TargetActor}
        specifiedDistance {defaultValue=750}
       }
      }
     ]
     properties {
      actionObjectKeyName=AIEvaluator_Healing_Object
      executionStatusKeyName=AIEvaluator_Healing_ExecutionStatus
      nodeName=Use Heal Item
      finishAfterItemUse {defaultValue=false}
      inputHoldDuration {key=None}
      maximumWaitTimeBetweenUses {defaultValue=1}
      resetActionObjectKey {defaultValue=true}
     }
    }
    Wait {
     class=BTTask_Wait
     decorators [Distance to Bunker Position {
       class=FortBTDecorator_DistanceBetween
       properties {
        flowAbortMode=Self
        nodeName=Distance to Bunker Position
        operator=Less
        blackboardKeyB {selectedKeyName=AIEvaluator_Bunker_Destination}
        specifiedDistance {defaultValue=125}
       }
      }]
     properties {randomDeviation {defaultValue=1}, waitTime {defaultValue=5}}
    }
   ]
   decorators []
   properties {
    children [
     {
      childComposite=null
      childTask=&self:FortAthenaBTTask_BotMoveTo_0
      decorators [&self:FortBTDecorator_DistanceBetween_1]
     }
     {
      childComposite=null
      childTask=&self:FortAthenaBTTask_UseItem_1
      decorators [
       &self:BTDecorator_Blackboard_1
       &self:FortBTDecorator_DistanceBetween_7
       &self:FortBTDecorator_DistanceBetween_8
      ]
     }
     {
      childComposite=null
      childTask=&self:BTTask_Wait_0
      decorators [&self:FortBTDecorator_DistanceBetween_6]
     }
    ]
    services [
     &self:FortAthenaBTService_AIEvaluator_3
     &self:FortAthenaBTService_AIEvaluator_4
     &self:FortAthenaBTService_AIEvaluator_33
    ]
   }
   services [
    Focus Scan Around or Heard Sound {
     class=FortAthenaBTService_AIEvaluator
     properties {
      interval=0.5
      nodeName=Focus Scan Around or Heard Sound
      aIEvaluatorClass {defaultValue=/Game/Athena/AI/Phoebe/Evaluators/Phoebe_HandleFocusing_FocusInvestigation.Phoebe_HandleFocusing_FocusInvestigation_C}
     }
    }
    Evaluating...Evasive Maneuvers {
     class=FortAthenaBTService_AIEvaluator
     properties {
      interval=0.5
      nodeName=Evaluating...Evasive Maneuvers
      aIEvaluatorClass {defaultValue=/Game/Athena/AI/Phoebe/Evaluators/Phoebe_EvasiveManeuvers_Cautious.Phoebe_EvasiveManeuvers_Cautious_C}
     }
    }
    Evaluating... Heal {
     class=FortAthenaBTService_AIEvaluator
     properties {
      interval=1000000000
      nodeName=Evaluating... Heal
      randomDeviation=0
      aIEvaluatorClass {defaultValue=/Game/Athena/AI/Phoebe/Evaluators/Phoebe_Evaluator_Healing.Phoebe_Evaluator_Healing_C}
     }
    }
   ]
  }
 ]
 decorators [Loop {class=BTDecorator_Loop}]
 properties {
  children [
   {
    childComposite=null
    childTask=&self:FortAthenaBTTask_Build_0
    decorators [&self:BTDecorator_Blackboard_26]
   }
   {childComposite=&self:BTComposite_Selector_1, childTask=null, decorators []}
  ]
  services []
 }
 services []
}
