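# Behavior tree for Phoebe AI bots in the DBNO (down but not out) state, using BB_Phoebe.
# An injected in-vehicle behavior takes priority over a sequence that opens with a short wait, cut short by detected danger.
# A general injection point then precedes a looping selector: flee to safety, move to the DBNO destination, or wait.
# Evaluator services feed danger, movement and threat-glance focus, while injector services supply both injection points.
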BehaviorTree /Game/Athena/AI/Phoebe/BehaviorTrees/BT_Phoebe_DBNO
blackboard=&BB_Phoebe
root {
 name=Selector
 class=BTComposite_Selector
 children [
  Injection Point DBNO.InVehicle {
   class=FortBTTask_InjectionPoint
   children []
   decorators [Can Do Injected Behavior {
     class=FortBTDecorator_CanDoInjectedBehavior
     properties {
      arithmeticOperation=GreaterOrEqual
      bInverseCondition=false
      basicOperation=Set
      floatValue=0
      flowAbortMode=Both
      intValue=3
      nodeName=Can Do Injected Behavior
      notifyObserver=ResultChange
      stringValue=ExecutionPending
      textOperation=Equal
      blackboardKey {bNoneIsAllowedValue=false, selectedKeyName=AIEvaluator_Injection_DBNO_InVehicle}
     }
    }]
   properties {
    bIgnoreRestartSelf=false
    nodeName=Injection Point DBNO.InVehicle
    injectedAction {baseClass=&Object, defaultValue=null, key=Injection_DBNO_InVehicle_ActionObject}
    injectionTag {tagName=Behavior.Injection.DBNO.InVehicle}
   }
   services []
  }
  Sequence {
   class=BTComposite_Sequence
   children [
    Wait {
     class=BTTask_Wait
     decorators [
      No danger detected? {
       class=BTDecorator_Blackboard
       properties {
        arithmeticOperation=Less
        flowAbortMode=Self
        intValue=3
        nodeName=No danger detected?
        stringValue=EExecutionStatus::ExecutionPending
        blackboardKey {selectedKeyName=AIEvaluator_DangerZoneDetected_ExecutionStatus}
       }
      }
      Force Success {class=BTDecorator_ForceSuccess}
     ]
     properties {randomDeviation {defaultValue=0.5}, waitTime {defaultValue=0.5}}
    }
    Selector {
     class=BTComposite_Selector
     children [
      Injection Point General.DBNO {
       class=FortBTTask_InjectionPoint
       children []
       decorators [Can Do Injected Behavior {
         class=FortBTDecorator_CanDoInjectedBehavior
         properties {
          arithmeticOperation=GreaterOrEqual
          bInverseCondition=false
          basicOperation=Set
          floatValue=0
          flowAbortMode=Both
          intValue=3
          nodeName=Can Do Injected Behavior
          notifyObserver=ResultChange
          stringValue=ExecutionPending
          textOperation=Equal
          blackboardKey {bNoneIsAllowedValue=false, selectedKeyName=AIEvaluator_Injection_General_DBNO}
         }
        }]
       properties {
        bIgnoreRestartSelf=false
        nodeName=Injection Point General.DBNO
        injectedAction {baseClass=&Object, defaultValue=null, key=Injection_General_DBNO_ActionObject}
        injectionTag {tagName=Behavior.Injection.General.DBNO}
       }
       services []
      }
      Selector {
       class=BTComposite_Selector
       children [
        Move away from danger {
         class=FortAthenaBTTask_BotMoveTo
         decorators [Danger detected? {
           class=BTDecorator_Blackboard
           properties {
            arithmeticOperation=GreaterOrEqual
            intValue=3
            nodeName=Danger detected?
            stringValue=EExecutionStatus::ExecutionPending
            blackboardKey {selectedKeyName=AIEvaluator_DangerZoneDetected_ExecutionStatus}
           }
          }]
         properties {
          executionStatusKeyName=AIEvaluator_DangerZoneDetected_ExecutionStatus
          movementResultKeyName=None
          nodeName=Move away from danger
          acceptableRadius {defaultValue=0.2}
          allowRandomWobble {defaultValue=false}
          bAllowStrafe {defaultValue=true}
          bTrackMovingGoal {defaultValue=false}
          blackboardKey {selectedKeyName=AIEvaluator_DangerZoneDetected_SafeLocation}
          observedBlackboardValueTolerance {defaultValue=4.75}
         }
        }
        Bot Move To {
         class=FortAthenaBTTask_BotMoveTo
         decorators [Can Do DBNO Move {
           class=BTDecorator_Blackboard
           properties {
            arithmeticOperation=GreaterOrEqual
            intValue=3
            nodeName=Can Do DBNO Move
            stringValue=EExecutionStatus::ExecutionPending
            blackboardKey {selectedKeyName=AIEvaluator_DBNO_ExecutionStatus}
           }
          }]
         properties {
          movementResultKeyName=AIEvaluator_DBNO_MovementState
          nodeName=Bot Move To
          acceptableRadius {defaultValue=150}
          allowRandomWobble {defaultValue=false}
          bAllowStrafe {defaultValue=false}
          blackboardKey {selectedKeyName=AIEvaluator_DBNO_Destination}
          filterClass {defaultValue=/Game/Athena/AI/Phoebe/NavFilter/NavFilter_Phoebe_DBNO.NavFilter_Phoebe_DBNO_C}
          observedBlackboardValueTolerance {defaultValue=4.75}
         }
        }
        Wait {class=BTTask_Wait, properties {waitTime {defaultValue=0.5}}}
       ]
       decorators [Loop {class=BTDecorator_Loop}]
       properties {
        children [
         {
          childComposite=null
          childTask=&self:FortAthenaBTTask_BotMoveTo_2
          decorators [&self:BTDecorator_Blackboard_3]
         }
         {
          childComposite=null
          childTask=&self:FortAthenaBTTask_BotMoveTo_1
          decorators [&self:BTDecorator_Blackboard_5]
         }
         {childComposite=null, childTask=&self:BTTask_Wait_0, decorators []}
        ]
        services [
         &self:FortAthenaBTService_AIEvaluator_8
         &self:FortAthenaBTService_AIEvaluator_0
        ]
       }
       services [
        Evaluating...DBNO Move {
         class=FortAthenaBTService_AIEvaluator
         properties {
          bCallTickOnSearchStart=true
          interval=0.2
          nodeName=Evaluating...DBNO Move
          randomDeviation=0
          aIEvaluatorClass {defaultValue=/Game/Athena/AI/Phoebe/Evaluators/Phoebe_Evaluator_DBNO.Phoebe_Evaluator_DBNO_C}
          injectionTag {tagName=Behavior.Injection.Evaluator.DBNO}
         }
        }
        Evaluating... DBNO Focus {
         class=FortAthenaBTService_AIEvaluator
         properties {
          interval=0.2
          nodeName=Evaluating... DBNO Focus
          randomDeviation=0
          aIEvaluatorClass {defaultValue=/Game/Athena/AI/Phoebe/Evaluators/Phoebe_HandleFocusing_GlanceAtThreat.Phoebe_HandleFocusing_GlanceAtThreat_C}
         }
        }
       ]
      }
     ]
     decorators []
     properties {
      children [
       {
        childComposite=null
        childTask=&self:FortBTTask_InjectionPoint_1
        decorators [&self:FortBTDecorator_CanDoInjectedBehavior_1]
       }
       {
        childComposite=&self:BTComposite_Selector_5
        childTask=null
        decorators [&self:BTDecorator_Loop_3]
       }
      ]
      services []
     }
     services []
    }
   ]
   decorators []
   properties {
    children [
     {
      childTask=&self:BTTask_Wait_3
      decorators [
       &self:BTDecorator_Blackboard_4
       &self:BTDecorator_ForceSuccess_0
      ]
     }
     {childComposite=&self:BTComposite_Selector_0, childTask=null}
    ]
    services [
     &self:FortAthenaBTService_AIEvaluator_13
     &self:FortBTService_AIEvaluatorsInjector_0
    ]
   }
   services [
    Evaluating...Danger {
     class=FortAthenaBTService_AIEvaluator
     properties {
      interval=1000000000
      nodeName=Evaluating...Danger
      randomDeviation=0
      aIEvaluatorClass {defaultValue=/Game/Athena/AI/Phoebe/Evaluators/Phoebe_Evaluator_DangerDetection.Phoebe_Evaluator_DangerDetection_C}
     }
    }
    Behavior Injector {
     class=FortBTService_AIEvaluatorsInjector
     properties {
      bCallTickOnSearchStart=false
      bRestartTimerOnEachActivation=false
      interval=0.5
      nodeName=Behavior Injector
      randomDeviation=0
      injectionTagsKeys [{
        actionInterfaceObjectKeyName=Injection_General_DBNO_ActionObject
        injectionKeyName=AIEvaluator_Injection_General_DBNO
        injectionTag {tagName=Behavior.Injection.General.DBNO}
       }]
     }
    }
   ]
  }
 ]
 decorators []
 properties {
  children [
   {
    childComposite=null
    childTask=&self:FortBTTask_InjectionPoint_0
    decorators [&self:FortBTDecorator_CanDoInjectedBehavior_0]
   }
   {childComposite=&self:BTComposite_Sequence_5, childTask=null, decorators []}
  ]
  services [&self:FortBTService_AIEvaluatorsInjector_1]
 }
 services [Behavior Injector {
   class=FortBTService_AIEvaluatorsInjector
   properties {
    bCallTickOnSearchStart=false
    bRestartTimerOnEachActivation=false
    interval=0.5
    nodeName=Behavior Injector
    randomDeviation=0.1
    injectionTagsKeys [{
      actionInterfaceObjectKeyName=Injection_DBNO_InVehicle_ActionObject
      injectionKeyName=AIEvaluator_Injection_DBNO_InVehicle
      injectionTag {tagName=Behavior.Injection.DBNO.InVehicle}
     }]
   }
  }]
}
