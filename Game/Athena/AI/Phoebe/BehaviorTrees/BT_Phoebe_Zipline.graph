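# Behavior tree that lets Phoebe AI player bots travel to and use ziplines, gated by the CanUseZiplines behavior control.
# A Zipline service on the root selector shares target, entry location and execution-status blackboard keys with the branches.
# The selector runs the Ziplining task when usage is allowed, otherwise a sequence that moves to the entry point and interacts.
# Blackboard status checks gate each branch, and the interaction aborts itself when its permission drops.
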BehaviorTree /Game/Athena/AI/Phoebe/BehaviorTrees/BT_Phoebe_Zipline
blackboard=&BB_Phoebe
root {
 name=Selector
 class=BTComposite_Selector
 children [
  Ziplining {
   class=FortAthenaBTTask_Zipline
   children []
   decorators [Is on Zipline {
     class=BTDecorator_Blackboard
     properties {
      arithmeticOperation=GreaterOrEqual
      intValue=4
      nodeName=Is on Zipline
      stringValue=EExecutionStatus::ExecutionAllowed
      blackboardKey {selectedKeyName=AIEvaluator_Zipline_UsageExecutionStatus}
     }
    }]
   properties {
    bIgnoreRestartSelf=false
    nodeName=Ziplining
    usageExecutionStatusName=AIEvaluator_Zipline_UsageExecutionStatus
    ziplineTargetName=AIEvaluator_Zipline_Target
   }
   services []
  }
  Sequence {
   class=BTComposite_Sequence
   children [
    Move To Zipline {
     class=FortAthenaBTTask_BotMoveTo
     decorators []
     properties {
      movementResultKeyName=AIEvaluator_Zipline_MovementState
      nodeName=Move To Zipline
      acceptableRadius {defaultValue=50}
      allowRandomWobble {defaultValue=false}
      bAllowPartialPath {defaultValue=false}
      bAllowStrafe {defaultValue=false}
      bProjectGoalLocation {defaultValue=false}
      bReachTestIncludesGoalRadius {defaultValue=false}
      bTrackMovingGoal {defaultValue=false}
      blackboardKey {selectedKeyName=AIEvaluator_Zipline_EntryLocation}
      observedBlackboardValueTolerance {defaultValue=200}
     }
    }
    Zipline Interaction {
     class=FortAthenaBTTask_Interact
     children []
     decorators [Can Interact with Zipline {
       class=BTDecorator_Blackboard
       properties {
        arithmeticOperation=GreaterOrEqual
        flowAbortMode=Self
        intValue=4
        nodeName=Can Interact with Zipline
        stringValue=EExecutionStatus::ExecutionAllowed
        blackboardKey {selectedKeyName=AIEvaluator_Zipline_InteractionExecutionStatus}
       }
      }]
     properties {
      attemptInterval=0.25
      bIgnoreRestartSelf=false
      bShouldFocusOnInteraction=false
      focalPointName=AIEvaluator_Global_FocalPoint
      interactActorName=AIEvaluator_Global_InteractActor
      jumpExecutionStatusName=AIEvaluator_Jump_ExecutionStatus
      maxInteractAttempts=5
      nodeName=Zipline Interaction
      weaponTriggerMeleeName=AIEvaluator_WeaponTriggerMelee_ExecutionStatus
      interactContextInfoKeySelector {bNoneIsAllowedValue=false, selectedKeyName=AIEvaluator_Zipline_InteractionContextInfo}
      interactExecutionStatusKeySelector {bNoneIsAllowedValue=false, selectedKeyName=AIEvaluator_Zipline_InteractionExecutionStatus}
     }
     services []
    }
   ]
   decorators [Should Move to Zipline {
     class=BTDecorator_Blackboard
     properties {
      arithmeticOperation=GreaterOrEqual
      intValue=4
      nodeName=Should Move to Zipline
      stringValue=EExecutionStatus::ExecutionAllowed
      blackboardKey {selectedKeyName=AIEvaluator_Zipline_MoveExecutionStatus}
     }
    }]
   properties {children [
     {childTask=&self:FortAthenaBTTask_BotMoveTo_1}
     {
      childTask=&self:FortAthenaBTTask_Interact_1
      decorators [&self:BTDecorator_Blackboard_31]
     }
    ]}
  }
 ]
 decorators ["Behavior Controls = CanUseZiplines" {
   class=FortAthenaBTDecorator_BehaviorControls
   properties {bInverseCondition=false, behaviorTreeBranch=CanUseZiplines, flowAbortMode=Both, nodeName="Behavior Controls = CanUseZiplines"}
  }]
 properties {
  children [
   {
    childComposite=null
    childTask=&self:FortAthenaBTTask_Zipline_0
    decorators [&self:BTDecorator_Blackboard_4]
   }
   {
    childComposite=&self:BTComposite_Sequence_1
    childTask=null
    decorators [&self:BTDecorator_Blackboard_17]
   }
  ]
  services [&self:FortAthenaBTService_Zipline_0]
 }
 services [Player Bot Zipline {
   class=FortAthenaBTService_Zipline
   properties {
    bCallTickOnSearchStart=false
    bRestartTimerOnEachActivation=false
    focalPointName=AIEvaluator_Global_FocalPoint
    interactionContextInfoName=AIEvaluator_Zipline_InteractionContextInfo
    interactionExecutionStatusName=AIEvaluator_Zipline_InteractionExecutionStatus
    interval=0.5
    moveExecutionStatusName=AIEvaluator_Zipline_MoveExecutionStatus
    moveToPathMovementStateName=AIEvaluator_Zipline_MovementState
    nodeName=Player Bot Zipline
    randomDeviation=0.1
    usageExecutionStatusName=AIEvaluator_Zipline_UsageExecutionStatus
    ziplineEntryLocationName=AIEvaluator_Zipline_EntryLocation
    ziplineExitLocationKeyName=AIEvaluator_Zipline_ExitLocation
    ziplineTargetName=AIEvaluator_Zipline_Target
   }
  }]
}
