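# Behavior tree that moves a Phoebe AI bot to its investigate destination, the last known position (LKP).
# A selector first runs the zipline subtree when either zipline move or usage execution status is at least ExecutionAllowed.
# Otherwise, or if that fails, it moves to the Investigate destination without strafing, wobble or moving-goal tracking, writing the movement state.
# Selector services periodically evaluate investigate zipline use and focus on heard sounds.
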BehaviorTree /Game/Athena/AI/Phoebe/BehaviorTrees/BT_Phoebe_MoveToLKP
blackboard=&BB_Phoebe
root {
 name=Selector
 class=BTComposite_Selector
 children [
  Run Zipline Behavior {
   class=BTTask_RunBehavior
   decorators [
    Blackboard Based Condition {
     class=BTDecorator_Blackboard
     properties {
      arithmeticOperation=GreaterOrEqual
      intValue=4
      nodeName=Blackboard Based Condition
      stringValue=EExecutionStatus::ExecutionAllowed
      blackboardKey {selectedKeyName=AIEvaluator_Zipline_MoveExecutionStatus}
     }
    }
    Blackboard Based Condition {
     class=BTDecorator_Blackboard
     properties {
      arithmeticOperation=GreaterOrEqual
      intValue=4
      nodeName=Blackboard Based Condition
      stringValue=EExecutionStatus::ExecutionAllowed
      blackboardKey {selectedKeyName=AIEvaluator_Zipline_UsageExecutionStatus}
     }
    }
   ]
   properties {behaviorAsset=&BT_Phoebe_Zipline, nodeName=Run Zipline Behavior}
  }
  Move To LKP {
   class=FortAthenaBTTask_BotMoveTo
   decorators []
   properties {
    movementResultKeyName=AIEvaluator_Investigate_MovementState
    nodeName=Move To LKP
    acceptableRadius {defaultValue=50}
    allowRandomWobble {defaultValue=false}
    bAllowStrafe {defaultValue=false}
    bReachTestIncludesAgentRadius {defaultValue=false}
    bReachTestIncludesGoalRadius {defaultValue=false}
    bTrackMovingGoal {defaultValue=false}
    blackboardKey {selectedKeyName=AIEvaluator_Investigate_Destination}
    observedBlackboardValueTolerance {defaultValue=4.75}
   }
  }
 ]
 decorators []
 properties {
  children [
   {
    childComposite=null
    childTask=&self:BTTask_RunBehavior_23
    decoratorOps [{number=2, operation=Or}, {number=0}, {number=1}]
    decorators [
     &self:BTDecorator_Blackboard_5
     &self:BTDecorator_Blackboard_6
    ]
   }
   {childComposite=null, childTask=&self:FortAthenaBTTask_BotMoveTo_3, decorators []}
  ]
  services [
   &self:FortAthenaBTService_AIEvaluator_19
   &self:FortAthenaBTService_AIEvaluator_29
  ]
 }
 services [
  Focus on heard sound {
   class=FortAthenaBTService_AIEvaluator
   properties {
    interval=0.5
    nodeName=Focus on heard sound
    aIEvaluatorClass {defaultValue=/Game/Athena/AI/Phoebe/Evaluators/Phoebe_HandleFocusing_FocusOnPerceivedSound.Phoebe_HandleFocusing_FocusOnPerceivedSound_C}
   }
  }
  Evaluate...Zipline {
   class=FortAthenaBTService_AIEvaluator
   properties {
    bCallTickOnSearchStart=true
    interval=0.5
    nodeName=Evaluate...Zipline
    aIEvaluatorClass {defaultValue=/Game/Athena/AI/Phoebe/Evaluators/Phoebe_Evaluator_Zipline_Investigate.Phoebe_Evaluator_Zipline_Investigate_C}
   }
  }
 ]
}
