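# EQS query for Phoebe AI bots that finds a valid navigable location around the querying bot, apparently for unstuck recovery.
# A spiral donut generator spreads sparse points from near to far around the bot, projecting them onto navmesh with a tall vertical range and the Phoebe nav filter.
# A pathfinding test filters and scores each point on a path from it to the Unstuck Reachable Goal context, using the same filter.
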EnvQuery /Game/Athena/AI/Phoebe/EQS/EQS_Phoebe_FindValidLocationAroundBot
options [{
  generator {
   name=EnvQueryGenerator_Donut_0
   class=EnvQueryGenerator_Donut
   properties {
    bUseSpiralPattern=true
    center=&EnvQueryContext_Querier
    innerRadius=300
    numberOfRings=5
    outerRadius=10000
    pointsPerRing=4
    projectionData {
     extentX=42
     navigationFilter=/Game/Athena/AI/Phoebe/NavFilter/NavFilter_Phoebe.NavFilter_Phoebe_C
     postProjectionVerticalOffset=0
     projectDown=10024
     projectUp=10024
    }
   }
  }
  tests [EnvQueryTest_Pathfinding_0 {
    class=EnvQueryTest_Pathfinding
    properties {
     context=&/EQS_Phoebe_Context_Unstuck_Reachable_Goal.EQS_Phoebe_Context_Unstuck_Reachable_Goal_C
     filterClass=&NavFilter_Phoebe_C
     testPurpose=FilterAndScore
     pathFromContext {dataBinding=null, dataField=BoolValue, defaultValue=false}
    }
   }]
 }]
