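# Behavior tree that frees stuck Phoebe bots, gated by the Unstuck execution status.
# A priority selector tries zipline subtree, water-avoiding move, undermining, steering, EQS-placed teleport, emote, then a short wait fallback.
# Blackboard execution statuses gate the first five branches, while Behavior Controls decorators also gate teleport and emote.
# Root services run focus and zipline evaluators, and copy an error status into water, teleport, undermine and steer keys on exit.
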BehaviorTree /Game/Athena/AI/Phoebe/BehaviorTrees/BT_Phoebe_UnstuckNew
blackboard=&BB_Phoebe
root {
 name=Unstuck Selector
 class=BTComposite_Selector
 children [
  Run Zipline Behavior {
   class=BTTask_RunBehavior
   decorators [
    Blackboard Based Condition {
     class=BTDecorator_Blackboard
     properties {
      arithmeticOperation=GreaterOrEqual
      intValue=4
      nodeName=Blackboard Based Condition
      stringValue=EExecutionStatus::ExecutionAllowed
      blackboardKey {selectedKeyName=AIEvaluator_Zipline_MoveExecutionStatus}
     }
    }
    Blackboard Based Condition {
     class=BTDecorator_Blackboard
     properties {
      arithmeticOperation=GreaterOrEqual
      intValue=4
      nodeName=Blackboard Based Condition
      stringValue=EExecutionStatus::ExecutionAllowed
      blackboardKey {selectedKeyName=AIEvaluator_Zipline_UsageExecutionStatus}
     }
    }
   ]
   properties {behaviorAsset=&BT_Phoebe_Zipline, nodeName=Run Zipline Behavior}
  }
  Bot Move To Avoid Water {
   class=FortAthenaBTTask_BotMoveTo
   decorators [Stuck In Water? {
     class=BTDecorator_Blackboard
     properties {
      arithmeticOperation=Equal
      flowAbortMode=None
      intValue=4
      nodeName=Stuck In Water?
      stringValue=EExecutionStatus::ExecutionAllowed
      blackboardKey {selectedKeyName=AIEvaluator_UnstuckAvoidWater_ExecutionStatus}
     }
    }]
   properties {
    movementResultKeyName=None
    nodeName=Bot Move To Avoid Water
    acceptableRadius {defaultValue=5}
    allowRandomWobble {defaultValue=false}
    bAllowStrafe {defaultValue=false}
    blackboardKey {selectedKeyName=AIEvaluator_UnstuckTeleport_ReachableGoal}
    filterClass {defaultValue=&NavFilter_Phoebe_AvoidWater_C}
    observedBlackboardValueTolerance {defaultValue=4.75}
   }
  }
  Undermine {
   class=FortAthenaBTTask_Undermine
   children []
   decorators [Can Undermine {
     class=BTDecorator_Blackboard
     properties {
      arithmeticOperation=GreaterOrEqual
      intValue=3
      nodeName=Can Undermine
      stringValue=EExecutionStatus::ExecutionPending
      blackboardKey {selectedKeyName=AIEvaluator_Undermine_ExecutionStatus}
     }
    }]
   properties {
    bIgnoreRestartSelf=false
    nodeName=Undermine
    undermineExecutionStatusKeyName=AIEvaluator_Undermine_ExecutionStatus
    undermineLocationImpactName=AIEvaluator_Undermine_Location
    undermineTargetKeyName=AIEvaluator_Undermine_Target
   }
   services []
  }
  Steer Movement {
   class=FortAthenaBTTask_SteerMovement
   children []
   decorators [Should Steer? {
     class=BTDecorator_Blackboard
     properties {
      arithmeticOperation=Equal
      intValue=4
      nodeName=Should Steer?
      stringValue=EExecutionStatus::ExecutionAllowed
      blackboardKey {selectedKeyName=AIEvaluator_UnstuckSteer_ExecutionStatus}
     }
    }]
   properties {
    bIgnoreRestartSelf=false
    bSetControlRotation=true
    nodeName=Steer Movement
    steerDirectionKeySelector {bNoneIsAllowedValue=false, selectedKeyName=AIEvaluator_UnstuckSteer_Direction}
   }
   services []
  }
  Phoebe Unstuck Teleport {
   class=FortAthenaBTTask_BotUnstuckTeleport
   children []
   decorators [
    Can Teleport {
     class=BTDecorator_Blackboard
     properties {
      arithmeticOperation=GreaterOrEqual
      flowAbortMode=None
      intValue=3
      nodeName=Can Teleport
      stringValue=EExecutionStatus::ExecutionPending
      blackboardKey {selectedKeyName=AIEvaluator_UnstuckTeleport_ExecutionStatus}
     }
    }
    "Behavior Controls = CanTeleportWhenStuck" {
     class=FortAthenaBTDecorator_BehaviorControls
     properties {bInverseCondition=false, behaviorTreeBranch=CanTeleportWhenStuck, flowAbortMode=Both, nodeName="Behavior Controls = CanTeleportWhenStuck"}
    }
   ]
   properties {
    bIgnoreRestartSelf=false
    canReachDestinationKeyName=AIEvaluator_UnstuckTeleport_ReachableGoal
    lastPartialPathCountKeyName=AIEvaluator_UnstuckTeleport_LastPartialCount
    lastPartialPathTimeKeyName=AIEvaluator_UnstuckTeleport_LastPartialPathTime
    nodeName=Phoebe Unstuck Teleport
    teleportExecutionStatusKeyName=AIEvaluator_UnstuckTeleport_ExecutionStatus
    eQSRequest {
     bUseBBKeyForQueryTemplate=false
     queryTemplate=/Game/Athena/AI/Phoebe/EQS/EQS_Phoebe_FindValidLocationAroundBot
     runMode=RandomBest25Pct
     eQSQueryBlackboardKey {bNoneIsAllowedValue=false, selectedKeyName=AIEvaluator_Global_LastKnownPosition}
     queryConfig []
    }
   }
   services []
  }
  Dance Baby {
   class=FortAthenaBTTask_PlayEmote
   children []
   decorators ["Behavior Controls = CanEmoteWhenStuck" {
     class=FortAthenaBTDecorator_BehaviorControls
     properties {bInverseCondition=false, behaviorTreeBranch=CanEmoteWhenStuck, flowAbortMode=Both, nodeName="Behavior Controls = CanEmoteWhenStuck"}
    }]
   properties {
    bIgnoreRestartSelf=false
    nodeName=Dance Baby
    playEmoteExecutionStatusKeyName=AIEvaluator_UnstuckPlayEmote_ExecutionStatus
    bespokeEmotes []
   }
   services []
  }
  Wait Unstuck {class=BTTask_Wait, properties {nodeName=Wait Unstuck, waitTime {defaultValue=2}}}
 ]
 decorators [Allow Execution and Reset AIEvaluator_Unstuck_ExecutionStatus {
   class=FortBTDecorator_ExecutionStatus
   properties {
    arithmeticOperation=GreaterOrEqual
    bInverseCondition=false
    basicOperation=Set
    floatValue=0
    flowAbortMode=Both
    intValue=3
    nodeName=Allow Execution and Reset AIEvaluator_Unstuck_ExecutionStatus
    notifyObserver=ResultChange
    stringValue=ExecutionPending
    textOperation=Equal
    blackboardKey {bNoneIsAllowedValue=false, selectedKeyName=AIEvaluator_Unstuck_ExecutionStatus}
    shouldUpdateExecutionStatus {defaultValue=true, key=None}
   }
  }]
 properties {
  nodeName=Unstuck Selector
  children [
   {
    childComposite=null
    childTask=&self:BTTask_RunBehavior_23
    decoratorOps [{number=2, operation=Or}, {number=0}, {number=1}]
    decorators [
     &self:BTDecorator_Blackboard_8
     &self:BTDecorator_Blackboard_9
    ]
   }
   {
    childComposite=null
    childTask=&self:FortAthenaBTTask_BotMoveTo_0
    decorators [&self:BTDecorator_Blackboard_3]
   }
   {
    childComposite=null
    childTask=&self:FortAthenaBTTask_Undermine_0
    decorators [&self:BTDecorator_Blackboard_0]
   }
   {
    childComposite=null
    childTask=&self:FortAthenaBTTask_SteerMovement_0
    decorators [&self:BTDecorator_Blackboard_2]
   }
   {
    childComposite=null
    childTask=&self:FortAthenaBTTask_BotUnstuckTeleport_3
    decorators [
     &self:BTDecorator_Blackboard_1
     &self:FortAthenaBTDecorator_BehaviorControls_5
    ]
   }
   {
    childComposite=null
    childTask=&self:FortAthenaBTTask_PlayEmote_0
    decorators [&self:FortAthenaBTDecorator_BehaviorControls_0]
   }
   {childComposite=null, childTask=&self:BTTask_Wait_0, decorators []}
  ]
  services [
   &self:FortAthenaBTService_CopyBlackboardVariable_14
   &self:FortAthenaBTService_CopyBlackboardVariable_15
   &self:FortAthenaBTService_CopyBlackboardVariable_16
   &self:FortAthenaBTService_CopyBlackboardVariable_17
   &self:FortAthenaBTService_AIEvaluator_8
   &self:FortAthenaBTService_AIEvaluator_9
  ]
 }
 services [
  Reset Avoid Water ExecutionStatus {
   class=FortAthenaBTService_CopyBlackboardVariable
   properties {
    bCallTickOnSearchStart=false
    bCopyOnBecomeRelevant=false
    bCopyOnCeaseRelevant=true
    bCopyWhenSourceValueChange=false
    bRestartTimerOnEachActivation=false
    interval=0.5
    nodeName=Reset Avoid Water ExecutionStatus
    randomDeviation=0.1
    destinationBlackboardKey {bNoneIsAllowedValue=false, selectedKeyName=AIEvaluator_UnstuckAvoidWater_ExecutionStatus}
    sourceBlackboardKey {bNoneIsAllowedValue=false, selectedKeyName=ExecutionStatus_Error}
   }
  }
  Reset Teleport ExecutionStatus {
   class=FortAthenaBTService_CopyBlackboardVariable
   properties {
    bCallTickOnSearchStart=false
    bCopyOnBecomeRelevant=false
    bCopyOnCeaseRelevant=true
    bCopyWhenSourceValueChange=false
    bRestartTimerOnEachActivation=false
    interval=0.5
    nodeName=Reset Teleport ExecutionStatus
    randomDeviation=0.1
    destinationBlackboardKey {bNoneIsAllowedValue=false, selectedKeyName=AIEvaluator_UnstuckTeleport_ExecutionStatus}
    sourceBlackboardKey {bNoneIsAllowedValue=false, selectedKeyName=ExecutionStatus_Error}
   }
  }
  Reset Undermine ExecutionStatus {
   class=FortAthenaBTService_CopyBlackboardVariable
   properties {
    bCallTickOnSearchStart=false
    bCopyOnBecomeRelevant=false
    bCopyOnCeaseRelevant=true
    bCopyWhenSourceValueChange=false
    bRestartTimerOnEachActivation=false
    interval=0.5
    nodeName=Reset Undermine ExecutionStatus
    randomDeviation=0.1
    destinationBlackboardKey {bNoneIsAllowedValue=false, selectedKeyName=AIEvaluator_Undermine_ExecutionStatus}
    sourceBlackboardKey {bNoneIsAllowedValue=false, selectedKeyName=ExecutionStatus_Error}
   }
  }
  Reset Steer Execution Status {
   class=FortAthenaBTService_CopyBlackboardVariable
   properties {
    bCallTickOnSearchStart=false
    bCopyOnBecomeRelevant=false
    bCopyOnCeaseRelevant=true
    bCopyWhenSourceValueChange=false
    bRestartTimerOnEachActivation=false
    interval=0.5
    nodeName=Reset Steer Execution Status
    randomDeviation=0.1
    destinationBlackboardKey {bNoneIsAllowedValue=false, selectedKeyName=AIEvaluator_UnstuckSteer_ExecutionStatus}
    sourceBlackboardKey {bNoneIsAllowedValue=false, selectedKeyName=ExecutionStatus_Error}
   }
  }
  Focus on target {
   class=FortAthenaBTService_AIEvaluator
   properties {interval=0.5, nodeName=Focus on target, aIEvaluatorClass {defaultValue=&FortAthenaAIBotEvaluator_HandleFocusing}}
  }
  Evaluate...Zipline {
   class=FortAthenaBTService_AIEvaluator
   properties {
    bCallTickOnSearchStart=true
    interval=0.5
    nodeName=Evaluate...Zipline
    aIEvaluatorClass {defaultValue=/Script/FortniteAIServer.FortAthenaAIBotEvaluator_Zipline}
   }
  }
 ]
}
