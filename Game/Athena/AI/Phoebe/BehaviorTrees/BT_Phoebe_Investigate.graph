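# Behavior tree for Phoebe AI bots that handles investigating by moving to an LKP, driven by AIEvaluator_Investigate blackboard keys.
# When the move-to-destination key passes its threshold, a Selector tries the unstuck subtree, then a strafe-enabled move gated by the Cautious key, then the MoveToLKP subtree.
# If that branch is blocked or fails, the root falls back to a short randomized wait.
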BehaviorTree /Game/Athena/AI/Phoebe/BehaviorTrees/BT_Phoebe_Investigate
blackboard=&BB_Phoebe
root {
 name=Selector
 class=BTComposite_Selector
 children [
  Selector {
   class=BTComposite_Selector
   children [
    "Run Unstuck Behavior " {
     class=BTTask_RunBehavior
     decorators []
     properties {behaviorAsset=&BT_Phoebe_UnstuckNew, nodeName="Run Unstuck Behavior "}
    }
    Move To LKP {
     class=FortAthenaBTTask_BotMoveTo
     decorators [Should be Cautious {
       class=BTDecorator_Blackboard
       properties {
        arithmeticOperation=Equal
        flowAbortMode=None
        intValue=0
        nodeName=Should be Cautious
        stringValue=""
        blackboardKey {selectedKeyName=AIEvaluator_Investigate_Cautious}
       }
      }]
     properties {
      movementResultKeyName=AIEvaluator_Investigate_MovementState
      nodeName=Move To LKP
      acceptableRadius {defaultValue=50}
      allowRandomWobble {defaultValue=false}
      bAllowStrafe {defaultValue=true}
      bReachTestIncludesAgentRadius {defaultValue=false}
      bReachTestIncludesGoalRadius {defaultValue=false}
      bTrackMovingGoal {defaultValue=false}
      blackboardKey {selectedKeyName=AIEvaluator_Investigate_Destination}
      observedBlackboardValueTolerance {defaultValue=4.75}
     }
    }
    Run MoveToLKP {
     class=BTTask_RunBehavior
     decorators []
     properties {behaviorAsset=&/BT_Phoebe_MoveToLKP, nodeName=Run MoveToLKP}
    }
   ]
   decorators [Should move to LKP {
     class=BTDecorator_Blackboard
     properties {
      arithmeticOperation=GreaterOrEqual
      intValue=2
      nodeName=Should move to LKP
      stringValue=""
      blackboardKey {selectedKeyName=AIEvaluator_Investigate_MoveToDestination}
     }
    }]
   properties {
    children [
     {childComposite=null, childTask=&self:BTTask_RunBehavior_0, decorators []}
     {
      childComposite=null
      childTask=&self:FortAthenaBTTask_BotMoveTo_0
      decorators [&self:BTDecorator_Blackboard_6]
     }
     {childComposite=null, childTask=&self:BTTask_RunBehavior_22, decorators []}
    ]
    services []
   }
   services []
  }
  Wait {class=BTTask_Wait, properties {randomDeviation {defaultValue=0.5}, waitTime {defaultValue=0.75}}}
 ]
 decorators []
 properties {
  children [
   {
    childComposite=&self:BTComposite_Selector_1
    childTask=null
    decorators [&self:BTDecorator_Blackboard_1]
   }
   {childComposite=null, childTask=&self:BTTask_Wait_9, decorators []}
  ]
  services []
 }
 services []
}
