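# Behavior tree for Phoebe AI bots that moves the bot toward its melee attack destination.
# A Selector runs a strafe-enabled, full-path-only bot move, gated by blackboard checks on the melee evaluator's move flag and destination.
# The destination check aborts the move when it fails, and the task reports through a movement-state key.
# A short Wait serves as the fallback when the move cannot run or fails.
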BehaviorTree /Game/Athena/AI/Phoebe/BehaviorTrees/BT_Phoebe_MoveToMelee
blackboard=&BB_Phoebe
root {
 name=Selector
 class=BTComposite_Selector
 children [
  Melee Move To {
   class=FortAthenaBTTask_BotMoveTo
   decorators [
    Should move toward Target {
     class=BTDecorator_Blackboard
     properties {
      arithmeticOperation=Equal
      intValue=0
      nodeName=Should move toward Target
      stringValue=""
      blackboardKey {selectedKeyName=AIEvaluator_MeleeAttack_MoveToDestination}
     }
    }
    Attack Destination Is Valid {
     class=BTDecorator_Blackboard
     properties {
      arithmeticOperation=Equal
      flowAbortMode=Self
      intValue=0
      nodeName=Attack Destination Is Valid
      stringValue=""
      blackboardKey {selectedKeyName=AIEvaluator_MeleeAttack_Destination}
     }
    }
   ]
   properties {
    movementResultKeyName=AIEvaluator_MeleeAttack_MovementState
    nodeName=Melee Move To
    acceptableRadius {defaultValue=100}
    allowRandomWobble {defaultValue=false}
    bAllowPartialPath {defaultValue=false}
    bAllowStrafe {defaultValue=true}
    bReachTestIncludesAgentRadius {defaultValue=false}
    blackboardKey {selectedKeyName=AIEvaluator_MeleeAttack_Destination}
    observedBlackboardValueTolerance {defaultValue=200}
   }
  }
  Wait {class=BTTask_Wait, properties {waitTime {defaultValue=2}}}
 ]
 decorators []
 properties {
  children [
   {
    childComposite=null
    childTask=&self:FortAthenaBTTask_BotMoveTo_5
    decorators [
     &self:BTDecorator_Blackboard_0
     &self:BTDecorator_Blackboard_2
    ]
   }
   {childComposite=null, childTask=&self:BTTask_Wait_0, decorators []}
  ]
  services []
 }
 services []
}
