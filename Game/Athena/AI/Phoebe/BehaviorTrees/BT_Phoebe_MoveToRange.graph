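# Behavior tree that moves a Phoebe AI bot toward its attack destination to get into range.
# A root selector tries the unstuck subtree first, then a movement selector, then a short Wait as fallback.
# A periodic evaluator service feeds zipline status keys; either one reaching the allowed threshold permits the zipline subtree.
# Otherwise the bot moves without strafing, gated by move request, valid destination (self-aborting) and no trap on path.
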BehaviorTree /Game/Athena/AI/Phoebe/BehaviorTrees/BT_Phoebe_MoveToRange
blackboard=&BB_Phoebe
root {
 name=Selector
 class=BTComposite_Selector
 children [
  "Run Unstuck Behavior " {
   class=BTTask_RunBehavior
   decorators []
   properties {behaviorAsset=&BT_Phoebe_UnstuckNew, nodeName="Run Unstuck Behavior "}
  }
  Selector {
   class=BTComposite_Selector
   children [
    Run Zipline Behavior {
     class=BTTask_RunBehavior
     decorators [
      Blackboard Based Condition {
       class=BTDecorator_Blackboard
       properties {
        arithmeticOperation=GreaterOrEqual
        intValue=4
        nodeName=Blackboard Based Condition
        stringValue=EExecutionStatus::ExecutionAllowed
        blackboardKey {selectedKeyName=AIEvaluator_Zipline_MoveExecutionStatus}
       }
      }
      Blackboard Based Condition {
       class=BTDecorator_Blackboard
       properties {
        arithmeticOperation=GreaterOrEqual
        intValue=4
        nodeName=Blackboard Based Condition
        stringValue=EExecutionStatus::ExecutionAllowed
        blackboardKey {selectedKeyName=AIEvaluator_Zipline_UsageExecutionStatus}
       }
      }
     ]
     properties {behaviorAsset=&BT_Phoebe_Zipline, nodeName=Run Zipline Behavior}
    }
    Range Move To {
     class=FortAthenaBTTask_BotMoveTo
     decorators [
      Should move toward Target {
       class=BTDecorator_Blackboard
       properties {
        arithmeticOperation=Equal
        intValue=0
        nodeName=Should move toward Target
        stringValue=""
        blackboardKey {selectedKeyName=AIEvaluator_Attack_MoveToDestination}
       }
      }
      Attack Destination Is Valid {
       class=BTDecorator_Blackboard
       properties {
        arithmeticOperation=Equal
        flowAbortMode=Self
        intValue=0
        nodeName=Attack Destination Is Valid
        stringValue=""
        blackboardKey {selectedKeyName=AIEvaluator_Attack_Destination}
       }
      }
      Has not trap on path {
       class=BTDecorator_Blackboard
       properties {
        arithmeticOperation=Less
        basicOperation=NotSet
        intValue=3
        nodeName=Has not trap on path
        stringValue=""
        blackboardKey {selectedKeyName=AIEvaluator_TrapOnPath_Actor}
       }
      }
     ]
     properties {
      movementResultKeyName=AIEvaluator_RangeAttack_MovementState
      nodeName=Range Move To
      acceptableRadius {defaultValue=50}
      allowRandomWobble {defaultValue=false}
      bAllowStrafe {defaultValue=false}
      blackboardKey {selectedKeyName=AIEvaluator_Attack_Destination}
      observedBlackboardValueTolerance {defaultValue=200}
     }
    }
   ]
   decorators []
   properties {
    children [
     {
      childComposite=null
      childTask=&self:BTTask_RunBehavior_23
      decoratorOps [{number=2, operation=Or}, {number=0}, {number=1}]
      decorators [
       &self:BTDecorator_Blackboard_5
       &self:BTDecorator_Blackboard_6
      ]
     }
     {
      childComposite=null
      childTask=&self:FortAthenaBTTask_BotMoveTo_5
      decorators [
       &self:BTDecorator_Blackboard_0
       &self:BTDecorator_Blackboard_2
       &self:BTDecorator_Blackboard_1
      ]
     }
    ]
    services [&self:FortAthenaBTService_AIEvaluator_0]
   }
   services [Evaluate... Zipline {
     class=FortAthenaBTService_AIEvaluator
     properties {
      bCallTickOnSearchStart=true
      interval=0.5
      nodeName=Evaluate... Zipline
      aIEvaluatorClass {defaultValue=/Game/Athena/AI/Phoebe/Evaluators/Phoebe_Evaluator_Zipline_MoveToRange.Phoebe_Evaluator_Zipline_MoveToRange_C}
     }
    }]
  }
  Wait {class=BTTask_Wait, properties {waitTime {defaultValue=1}}}
 ]
 decorators []
 properties {
  children [
   {childComposite=null, childTask=&self:BTTask_RunBehavior_0, decorators []}
   {childComposite=&self:BTComposite_Selector_1, childTask=null, decorators []}
   {childComposite=null, childTask=&self:BTTask_Wait_3, decorators []}
  ]
  services []
 }
 services []
}
