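# Retreat behavior tree for Phoebe AI bots, which moves the bot to a retreat destination and can interrupt to build defenses.
# A defensive-building evaluator feeds a high-priority Build task that aborts movement once execution is pending.
# Otherwise, a Simple Parallel runs urgent, strafing MoveTo with a cover nav filter under a sprinting evaluator.
# Its background branch contains only randomized waits, chosen by distance to the destination.
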BehaviorTree /Game/Athena/AI/Phoebe/BehaviorTrees/BT_Phoebe_Retreat
blackboard=&BB_Phoebe
root {
 name=Selector
 class=BTComposite_Selector
 children [
  Build Construction {
   class=FortAthenaBTTask_Build
   children []
   decorators [Build protection {
     class=BTDecorator_Blackboard
     properties {
      arithmeticOperation=GreaterOrEqual
      flowAbortMode=LowerPriority
      intValue=3
      nodeName=Build protection
      stringValue=EExecutionStatus::ExecutionPending
      blackboardKey {selectedKeyName=AIEvaluator_DefensiveBuilding_ExecutionStatus}
     }
    }]
   properties {
    bIgnoreRestartSelf=false
    executionStatusKeyName=AIEvaluator_DefensiveBuilding_ExecutionStatus
    focalPointName=AIEvaluator_Global_FocalPoint
    nodeName=Build Construction
   }
   services []
  }
  Simple Parallel {
   class=BTComposite_SimpleParallel
   children [
    Retreat MoveTo {
     class=FortAthenaBTTask_BotMoveTo
     decorators []
     properties {
      movementResultKeyName=AIEvaluator_Retreat_MovementState
      nodeName=Retreat MoveTo
      acceptableRadius {defaultValue=50}
      allowRandomWobble {defaultValue=false}
      bAllowStrafe {defaultValue=true}
      blackboardKey {selectedKeyName=AIEvaluator_Retreat_Destination}
      filterClass {defaultValue=/Game/Athena/AI/Phoebe/NavFilter/NavFilter_Phoebe_Cover.NavFilter_Phoebe_Cover_C}
      isUrgentMovement {defaultValue=true}
      observedBlackboardValueTolerance {defaultValue=4.75}
     }
    }
    Selector {
     class=BTComposite_Selector
     children [
      Wait {
       class=BTTask_Wait
       decorators [Distance Between {
         class=FortBTDecorator_DistanceBetween
         properties {
          flowAbortMode=None
          nodeName=Distance Between
          operator=LessOrEqual
          bCalculateAs2D {defaultValue=true}
          blackboardKeyB {selectedKeyName=AIEvaluator_Retreat_Destination}
          distanceCalculationUpdateRate {defaultValue=3}
          specifiedDistance {defaultValue=800}
         }
        }]
       properties {randomDeviation {defaultValue=1}, waitTime {defaultValue=3}}
      }
      Sequence {
       class=BTComposite_Sequence
       children [
        Wait {class=BTTask_Wait, properties {randomDeviation {defaultValue=3}, waitTime {defaultValue=6}}}
        Wait {class=BTTask_Wait, properties {randomDeviation {defaultValue=1}, waitTime {defaultValue=3}}}
       ]
       decorators [Loop {class=BTDecorator_Loop}]
       properties {children [
         {childTask=&self:BTTask_Wait_4}
         {childTask=&self:BTTask_Wait_3}
        ]}
      }
     ]
     decorators []
     properties {
      children [
       {
        childComposite=null
        childTask=&self:BTTask_Wait_0
        decorators [&self:FortBTDecorator_DistanceBetween_0]
       }
       {
        childComposite=&self:BTComposite_Sequence_3
        childTask=null
        decorators [&self:BTDecorator_Loop_0]
       }
      ]
      services []
     }
     services []
    }
   ]
   properties {
    children [
     {childTask=&self:FortAthenaBTTask_BotMoveTo_3}
     {childComposite=&self:BTComposite_Selector_0, childTask=null}
    ]
    services [&self:FortAthenaBTService_AIEvaluator_57]
   }
   services [Evaluate...Sprinting {
     class=FortAthenaBTService_AIEvaluator
     properties {
      interval=0.2
      nodeName=Evaluate...Sprinting
      aIEvaluatorClass {defaultValue=/Game/Athena/AI/Phoebe/Evaluators/Phoebe_Evaluator_Sprinting.Phoebe_Evaluator_Sprinting_C}
     }
    }]
  }
 ]
 decorators [Loop {class=BTDecorator_Loop}]
 properties {
  children [
   {
    childComposite=null
    childTask=&self:FortAthenaBTTask_Build_4
    decorators [&self:BTDecorator_Blackboard_39]
   }
   {childComposite=&self:BTComposite_SimpleParallel_4, childTask=null, decorators []}
  ]
  services [&self:FortAthenaBTService_AIEvaluator_4]
 }
 services [Evaluating...Defensive Building {
   class=FortAthenaBTService_AIEvaluator
   properties {
    interval=0.5
    nodeName=Evaluating...Defensive Building
    aIEvaluatorClass {defaultValue=/Script/FortniteAIServer.FortAthenaAIBotEvaluator_DefensiveBuilding}
   }
  }]
}
