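# Phoebe bot behavior tree for reviving a target actor, driven by BB_Phoebe blackboard keys from the Revive evaluator.
# A single sequence runs a Revive service with leash disabled, using the target, movement and interaction status keys.
# It moves the bot to the revive target without partial paths or strafing, then runs a focused, retrying interact task.
# A blackboard check on interaction execution status gates the interaction and aborts it when it fails.
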BehaviorTree /Game/Athena/AI/Phoebe/BehaviorTrees/BT_Phoebe_Revive
blackboard=&BB_Phoebe
root {
 name=Selector
 class=BTComposite_Selector
 children [Sequence {
   class=BTComposite_Sequence
   children [
    Revive Move To {
     class=FortAthenaBTTask_BotMoveTo
     decorators []
     properties {
      movementResultKeyName=AIEvaluator_Revive_MovementState
      nodeName=Revive Move To
      acceptableRadius {defaultValue=20}
      allowRandomWobble {defaultValue=false}
      bAllowPartialPath {defaultValue=false}
      bAllowStrafe {defaultValue=false}
      blackboardKey {selectedKeyName=AIEvaluator_Revive_Target}
      observedBlackboardValueTolerance {defaultValue=4.75}
     }
    }
    Revive Interaction {
     class=FortAthenaBTTask_Interact
     children []
     decorators [Can Interact {
       class=BTDecorator_Blackboard
       properties {
        arithmeticOperation=GreaterOrEqual
        flowAbortMode=Self
        intValue=3
        nodeName=Can Interact
        stringValue=""
        blackboardKey {selectedKeyName=AIEvaluator_Revive_InteractionExecutionStatus}
       }
      }]
     properties {
      attemptInterval=0.25
      bIgnoreRestartSelf=false
      bShouldFocusOnInteraction=true
      focalPointName=AIEvaluator_Global_FocalPoint
      interactActorName=AIEvaluator_Global_InteractActor
      jumpExecutionStatusName=AIEvaluator_Jump_ExecutionStatus
      maxInteractAttempts=5
      nodeName=Revive Interaction
      weaponTriggerMeleeName=AIEvaluator_WeaponTriggerMelee_ExecutionStatus
      interactContextInfoKeySelector {bNoneIsAllowedValue=false, selectedKeyName=AIEvaluator_Revive_InteractionContextInfo}
      interactExecutionStatusKeySelector {bNoneIsAllowedValue=false, selectedKeyName=AIEvaluator_Revive_InteractionExecutionStatus}
     }
     services []
    }
   ]
   decorators []
   properties {
    children [
     {childTask=&self:FortAthenaBTTask_BotMoveTo_0}
     {
      childTask=&self:FortAthenaBTTask_Interact_1
      decorators [&self:BTDecorator_Blackboard_28]
     }
    ]
    services [&self:FortAthenaBTService_Revive_0]
   }
   services [Executing...Revive {
     class=FortAthenaBTService_Revive
     properties {
      bCallTickOnSearchStart=false
      bDisableLeash=true
      bRestartTimerOnEachActivation=false
      executionStatusName=AIEvaluator_Revive_ExecutionStatus
      interactionContextInfoName=AIEvaluator_Revive_InteractionContextInfo
      interactionExecutionStatusName=AIEvaluator_Revive_InteractionExecutionStatus
      interval=0.5
      moveToPathMovementStateName=AIEvaluator_Revive_MovementState
      nodeName=Executing...Revive
      randomDeviation=0.1
      reviveTargetKeyName=AIEvaluator_Revive_Target
     }
    }]
  }]
 decorators []
 properties {
  children [{childComposite=&self:BTComposite_Sequence_0, childTask=null, decorators []}]
  services []
 }
 services []
}
