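# Behavior tree that moves Phoebe AI bots out of the storm, with a sprinting evaluator service at the root.
# The root selector tries an unstuck subtree first, then a movement selector whose service evaluates storm-specific zipline use.
# That selector runs the zipline subtree if either zipline move or usage is allowed.
# Otherwise, a simple parallel moves to the storm evaluator's destination while building defensively when allowed.
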BehaviorTree /Game/Athena/AI/Phoebe/BehaviorTrees/BT_Phoebe_Storm
blackboard=&BB_Phoebe
root {
 name=Escape Storm Move Selector
 class=BTComposite_Selector
 children [
  Run Unstuck Behavior {
   class=BTTask_RunBehavior
   decorators []
   properties {behaviorAsset=&BT_Phoebe_UnstuckNew, nodeName=Run Unstuck Behavior}
  }
  Selector {
   class=BTComposite_Selector
   children [
    Run Zipline Behavior {
     class=BTTask_RunBehavior
     decorators [
      Blackboard Based Condition {
       class=BTDecorator_Blackboard
       properties {
        arithmeticOperation=GreaterOrEqual
        intValue=4
        nodeName=Blackboard Based Condition
        stringValue=EExecutionStatus::ExecutionAllowed
        blackboardKey {selectedKeyName=AIEvaluator_Zipline_MoveExecutionStatus}
       }
      }
      Blackboard Based Condition {
       class=BTDecorator_Blackboard
       properties {
        arithmeticOperation=GreaterOrEqual
        intValue=4
        nodeName=Blackboard Based Condition
        stringValue=EExecutionStatus::ExecutionAllowed
        blackboardKey {selectedKeyName=AIEvaluator_Zipline_UsageExecutionStatus}
       }
      }
     ]
     properties {behaviorAsset=&BT_Phoebe_Zipline, nodeName=Run Zipline Behavior}
    }
    Simple Parallel {
     class=BTComposite_SimpleParallel
     children [
      Storm Move To {
       class=FortAthenaBTTask_BotMoveTo
       decorators []
       properties {
        movementResultKeyName=AIEvaluator_Storm_MovementState
        nodeName=Storm Move To
        acceptableRadius {defaultValue=50}
        allowRandomWobble {defaultValue=false}
        bAllowStrafe {defaultValue=true}
        bTrackMovingGoal {defaultValue=false}
        blackboardKey {selectedKeyName=AIEvaluator_Storm_Destination}
        observedBlackboardValueTolerance {defaultValue=200}
       }
      }
      Build Construction {
       class=FortAthenaBTTask_Build
       children []
       decorators [Blackboard Based Condition {
         class=BTDecorator_Blackboard
         properties {
          arithmeticOperation=GreaterOrEqual
          flowAbortMode=None
          intValue=4
          nodeName=Blackboard Based Condition
          stringValue=EExecutionStatus::ExecutionAllowed
          blackboardKey {selectedKeyName=AIEvaluator_DefensiveBuilding_ExecutionStatus}
         }
        }]
       properties {
        bIgnoreRestartSelf=false
        executionStatusKeyName=AIEvaluator_DefensiveBuilding_ExecutionStatus
        focalPointName=AIEvaluator_Global_FocalPoint
        nodeName=Build Construction
       }
       services []
      }
     ]
     properties {
      children [
       {childTask=&self:FortAthenaBTTask_BotMoveTo_1}
       {
        childTask=&self:FortAthenaBTTask_Build_0
        decorators [&self:BTDecorator_Blackboard_1]
       }
      ]
      services []
     }
     services []
    }
   ]
   decorators []
   properties {
    children [
     {
      childComposite=null
      childTask=&self:BTTask_RunBehavior_23
      decoratorOps [{number=2, operation=Or}, {number=0}, {number=1}]
      decorators [
       &self:BTDecorator_Blackboard_0
       &self:BTDecorator_Blackboard_3
      ]
     }
     {childComposite=&self:BTComposite_SimpleParallel_0, childTask=null, decorators []}
    ]
    services [&self:FortAthenaBTService_AIEvaluator_29]
   }
   services [Evaluate...Zipline {
     class=FortAthenaBTService_AIEvaluator
     properties {
      bCallTickOnSearchStart=true
      interval=0.5
      nodeName=Evaluate...Zipline
      aIEvaluatorClass {defaultValue=/Game/Athena/AI/Phoebe/Evaluators/Phoebe_Evaluator_Zipline_Storm.Phoebe_Evaluator_Zipline_Storm_C}
     }
    }]
  }
 ]
 decorators [Loop {class=BTDecorator_Loop}]
 properties {
  nodeName=Escape Storm Move Selector
  children [
   {childComposite=null, childTask=&self:BTTask_RunBehavior_0, decorators []}
   {childComposite=&self:BTComposite_Selector_0, childTask=null, decorators []}
  ]
  services [&self:FortAthenaBTService_AIEvaluator_57]
 }
 services [Evaluate...Sprinting {
   class=FortAthenaBTService_AIEvaluator
   properties {
    interval=0.2
    nodeName=Evaluate...Sprinting
    aIEvaluatorClass {defaultValue=/Game/Athena/AI/Phoebe/Evaluators/Phoebe_Evaluator_Sprinting.Phoebe_Evaluator_Sprinting_C}
   }
  }]
}
