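# Behavior tree for Phoebe AI bots in Athena that takes cover and heals, with a looping root selector.
# With a valid cover destination, it tries unstuck behavior, then moves to cover on a no-smashing nav filter, while evaluators drive focus and sprinting.
# Otherwise it uses a heal item if healing is possible without line of sight on the threat, else waits.
# A root healing evaluator with a huge interval feeds the heal branch.
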BehaviorTree /Game/Athena/AI/Phoebe/BehaviorTrees/BT_Phoebe_TakeCover
blackboard=&BB_Phoebe
root {
 name=Selector
 class=BTComposite_Selector
 children [
  BTComposite_Selector_23 {
   class=BTComposite_Selector
   children [
    "Run Unstuck Behavior " {
     class=BTTask_RunBehavior
     decorators []
     properties {behaviorAsset=&BT_Phoebe_UnstuckNew, nodeName="Run Unstuck Behavior "}
    }
    Move to Cover {
     class=FortAthenaBTTask_BotMoveTo
     decorators []
     properties {
      movementResultKeyName=AIEvaluator_TakeCover_MovementState
      nodeName=Move to Cover
      acceptableRadius {defaultValue=50}
      allowRandomWobble {defaultValue=true}
      bAllowPartialPath {defaultValue=false}
      bAllowStrafe {defaultValue=false}
      bReachTestIncludesAgentRadius {defaultValue=false}
      bReachTestIncludesGoalRadius {defaultValue=false}
      bTrackMovingGoal {defaultValue=false}
      blackboardKey {selectedKeyName=AIEvaluator_TakeCover_Destination}
      filterClass {defaultValue=/Script/FortniteAI.FortNavigationFilter_NoSmashing}
      observedBlackboardValueTolerance {defaultValue=100}
     }
    }
   ]
   decorators [Has Valid Cover Destination {
     class=BTDecorator_Blackboard
     properties {
      arithmeticOperation=Equal
      intValue=0
      nodeName=Has Valid Cover Destination
      stringValue=""
      blackboardKey {selectedKeyName=AIEvaluator_TakeCover_MoveToDestination}
     }
    }]
   properties {
    nodeName=""
    children [
     {childComposite=null, childTask=&self:BTTask_RunBehavior_0, decorators []}
     {childComposite=null, childTask=&self:FortAthenaBTTask_BotMoveTo_0, decorators []}
    ]
    services [
     &self:FortAthenaBTService_AIEvaluator_3
     &self:FortAthenaBTService_AIEvaluator_57
    ]
   }
   services [
    Focus for Take Cover {
     class=FortAthenaBTService_AIEvaluator
     properties {
      interval=0.5
      nodeName=Focus for Take Cover
      aIEvaluatorClass {defaultValue=/Game/Athena/AI/Phoebe/Evaluators/Phoebe_HandleFocusing_TakeCover.Phoebe_HandleFocusing_TakeCover_C}
     }
    }
    Evaluate...Sprinting {
     class=FortAthenaBTService_AIEvaluator
     properties {
      interval=0.2
      nodeName=Evaluate...Sprinting
      aIEvaluatorClass {defaultValue=/Game/Athena/AI/Phoebe/Evaluators/Phoebe_Evaluator_Sprinting.Phoebe_Evaluator_Sprinting_C}
     }
    }
   ]
  }
  Use Heal Item {
   class=FortAthenaBTTask_UseItem
   decorators [
    Can Heal {
     class=BTDecorator_Blackboard
     properties {
      arithmeticOperation=GreaterOrEqual
      intValue=3
      nodeName=Can Heal
      stringValue=EExecutionStatus::ExecutionPending
      blackboardKey {selectedKeyName=AIEvaluator_Healing_ExecutionStatus}
     }
    }
    No LoS On Threat {
     class=BTDecorator_Blackboard
     properties {
      arithmeticOperation=GreaterOrEqual
      basicOperation=NotSet
      intValue=3
      nodeName=No LoS On Threat
      stringValue=""
      blackboardKey {selectedKeyName=AIEvaluator_Global_HasLoSOnThreat}
     }
    }
   ]
   properties {
    actionObjectKeyName=AIEvaluator_Healing_Object
    executionStatusKeyName=AIEvaluator_Healing_ExecutionStatus
    nodeName=Use Heal Item
    finishAfterItemUse {defaultValue=false}
    inputHoldDuration {key=None}
    maximumWaitTimeBetweenUses {defaultValue=1}
    resetActionObjectKey {defaultValue=true}
   }
  }
  Wait {class=BTTask_Wait, properties {waitTime {defaultValue=5}}}
 ]
 decorators [Loop {class=BTDecorator_Loop}]
 properties {
  children [
   {
    childComposite=&self:BTComposite_Selector_23
    childTask=null
    decorators [&self:BTDecorator_Blackboard_3]
   }
   {
    childComposite=null
    childTask=&self:FortAthenaBTTask_UseItem_1
    decorators [
     &self:BTDecorator_Blackboard_42
     &self:BTDecorator_Blackboard_7
    ]
   }
   {childComposite=null, childTask=&self:BTTask_Wait_0, decorators []}
  ]
  services [&self:FortAthenaBTService_AIEvaluator_33]
 }
 services [Evaluate... Heal {
   class=FortAthenaBTService_AIEvaluator
   properties {
    interval=1000000000
    nodeName=Evaluate... Heal
    randomDeviation=0
    aIEvaluatorClass {defaultValue=/Game/Athena/AI/Phoebe/Evaluators/Phoebe_Evaluator_Healing.Phoebe_Evaluator_Healing_C}
   }
  }]
}
