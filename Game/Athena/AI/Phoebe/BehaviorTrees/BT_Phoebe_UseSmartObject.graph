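# Behavior tree that lets a Phoebe AI bot travel to and use a SmartObject.
# The root runs a SmartObject service and is gated by execution status and the CanUseSmartObjects behavior control, both aborting in both directions.
# When ShouldMove is set, a selector tries the unstuck subtree, then urgency-based MoveTo tasks, urgent at High or above; otherwise movement succeeds immediately.
# The sequence then rotates the bot to the destination rotation and uses the SmartObject.
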BehaviorTree /Game/Athena/AI/Phoebe/BehaviorTrees/BT_Phoebe_UseSmartObject
blackboard=&BB_Phoebe
root {
 name=Selector
 class=BTComposite_Selector
 children [Sequence {
   class=BTComposite_Sequence
   children [
    Selector {
     class=BTComposite_Selector
     children [
      Selector {
       class=BTComposite_Selector
       children [
        Run Unstuck Behavior {
         class=BTTask_RunBehavior
         decorators []
         properties {behaviorAsset=&BT_Phoebe_UnstuckNew, nodeName=Run Unstuck Behavior}
        }
        MoveTo SmartObject {
         class=FortAthenaBTTask_BotMoveTo
         decorators ["Urgency <= Low" {
           class=BTDecorator_Blackboard
           properties {
            arithmeticOperation=LessOrEqual
            intValue=0
            nodeName="Urgency <= Low"
            stringValue=Low
            blackboardKey {selectedKeyName=AIEvaluator_SmartObject_Urgency}
           }
          }]
         properties {
          movementResultKeyName=AIEvaluator_SmartObject_MovementState
          nodeName=MoveTo SmartObject
          acceptableRadius {defaultValue=10}
          allowRandomWobble {defaultValue=false}
          bAllowPartialPath {defaultValue=false}
          bAllowStrafe {defaultValue=true}
          bReachTestIncludesAgentRadius {defaultValue=false}
          bReachTestIncludesGoalRadius {defaultValue=false}
          bTrackMovingGoal {defaultValue=false}
          blackboardKey {selectedKeyName=AIEvaluator_SmartObject_Destination}
          observedBlackboardValueTolerance {defaultValue=50}
         }
        }
        MoveTo SmartObject {
         class=FortAthenaBTTask_BotMoveTo
         decorators ["Urgency = Normal" {
           class=BTDecorator_Blackboard
           properties {
            arithmeticOperation=Equal
            intValue=1
            nodeName="Urgency = Normal"
            stringValue=Normal
            blackboardKey {selectedKeyName=AIEvaluator_SmartObject_Urgency}
           }
          }]
         properties {
          movementResultKeyName=AIEvaluator_SmartObject_MovementState
          nodeName=MoveTo SmartObject
          acceptableRadius {defaultValue=10}
          allowRandomWobble {defaultValue=false}
          bAllowPartialPath {defaultValue=false}
          bAllowStrafe {defaultValue=true}
          bReachTestIncludesAgentRadius {defaultValue=false}
          bReachTestIncludesGoalRadius {defaultValue=false}
          bTrackMovingGoal {defaultValue=false}
          blackboardKey {selectedKeyName=AIEvaluator_SmartObject_Destination}
          observedBlackboardValueTolerance {defaultValue=50}
         }
        }
        "MoveTo SmartObject (Urgent)" {
         class=FortAthenaBTTask_BotMoveTo
         decorators ["Urgency >= High" {
           class=BTDecorator_Blackboard
           properties {
            arithmeticOperation=GreaterOrEqual
            intValue=2
            nodeName="Urgency >= High"
            stringValue=High
            blackboardKey {selectedKeyName=AIEvaluator_SmartObject_Urgency}
           }
          }]
         properties {
          movementResultKeyName=AIEvaluator_SmartObject_MovementState
          nodeName="MoveTo SmartObject (Urgent)"
          acceptableRadius {defaultValue=10}
          allowRandomWobble {defaultValue=false}
          bAllowPartialPath {defaultValue=false}
          bAllowStrafe {defaultValue=true}
          bReachTestIncludesAgentRadius {defaultValue=false}
          bReachTestIncludesGoalRadius {defaultValue=false}
          bTrackMovingGoal {defaultValue=false}
          blackboardKey {selectedKeyName=AIEvaluator_SmartObject_Destination}
          isUrgentMovement {defaultValue=true}
          observedBlackboardValueTolerance {defaultValue=50}
         }
        }
       ]
       decorators [Should move to SmartObject? {
         class=BTDecorator_Blackboard
         properties {
          arithmeticOperation=Equal
          flowAbortMode=Self
          intValue=0
          nodeName=Should move to SmartObject?
          stringValue=""
          blackboardKey {selectedKeyName=AIEvaluator_SmartObject_ShouldMove}
         }
        }]
       properties {
        children [
         {childComposite=null, childTask=&self:BTTask_RunBehavior_5, decorators []}
         {
          childComposite=null
          childTask=&self:FortAthenaBTTask_BotMoveTo_1
          decorators [&self:BTDecorator_Blackboard_6]
         }
         {
          childComposite=null
          childTask=&self:FortAthenaBTTask_BotMoveTo_2
          decorators [&self:BTDecorator_Blackboard_7]
         }
         {
          childComposite=null
          childTask=&self:FortAthenaBTTask_BotMoveTo_3
          decorators [&self:BTDecorator_Blackboard_0]
         }
        ]
        services []
       }
       services []
      }
      BTTask_AlwaysSuccess {
       class=FortBTTask_AlwaysSucceed
       children []
       decorators [Blackboard Based Condition {
         class=BTDecorator_Blackboard
         properties {
          arithmeticOperation=Equal
          basicOperation=NotSet
          flowAbortMode=Self
          intValue=0
          nodeName=Blackboard Based Condition
          stringValue=""
          blackboardKey {selectedKeyName=AIEvaluator_SmartObject_ShouldMove}
         }
        }]
       properties {bIgnoreRestartSelf=false, nodeName=BTTask_AlwaysSuccess}
       services []
      }
     ]
     decorators []
     properties {
      children [
       {
        childComposite=&self:BTComposite_Selector_0
        childTask=null
        decorators [&self:BTDecorator_Blackboard_9]
       }
       {
        childComposite=null
        childTask=&self:BTTask_AlwaysSuccess_C_1
        decorators [&self:BTDecorator_Blackboard_2]
       }
      ]
      services []
     }
     services []
    }
    Rotate to face BB entry {
     class=BTTask_RotateToFaceBBEntry
     decorators []
     properties {blackboardKey {selectedKeyName=AIEvaluator_SmartObject_DestinationRotation}, precision {defaultValue=10}}
    }
    UseSmartObject {
     class=FortAthenaBTTask_UseSmartObject
     children []
     decorators []
     properties {
      bHandleAbortWithSoftDisable=false
      bIgnoreRestartSelf=false
      nodeName=UseSmartObject
      smartObjectDestinationRotationKeyName=AIEvaluator_SmartObject_DestinationRotation
      smartObjectsStatusKeyName=None
      addFocusAndWait {defaultValue=false, key=None}
      evaluationTag {tagName=None}
     }
     services []
    }
   ]
   decorators []
   properties {children [
     {childComposite=&self:BTComposite_Selector_1, childTask=null}
     {childTask=&self:BTTask_RotateToFaceBBEntry_0}
     {childTask=&self:FortAthenaBTTask_UseSmartObject_1}
    ]}
  }]
 decorators [
  Is Execution Allowed {
   class=FortBTDecorator_ExecutionStatus
   properties {
    arithmeticOperation=GreaterOrEqual
    bInverseCondition=false
    basicOperation=Set
    floatValue=0
    flowAbortMode=Both
    intValue=3
    nodeName=Is Execution Allowed
    notifyObserver=ResultChange
    stringValue=ExecutionPending
    textOperation=Equal
    blackboardKey {bNoneIsAllowedValue=false, selectedKeyName=AIEvaluator_SmartObject_ExecutionStatus}
    shouldUpdateExecutionStatus {defaultValue=true, key=None}
   }
  }
  "Behavior Controls = CanUseSmartObjects" {
   class=FortAthenaBTDecorator_BehaviorControls
   properties {bInverseCondition=false, behaviorTreeBranch=CanUseSmartObjects, flowAbortMode=Both, nodeName="Behavior Controls = CanUseSmartObjects"}
  }
 ]
 properties {
  children [{childComposite=&self:BTComposite_Sequence_0, childTask=null, decorators []}]
  services [&self:FortAthenaBTService_SmartObject_0]
 }
 services [Executing...SmartObject {
   class=FortAthenaBTService_SmartObject
   properties {
    bCallTickOnSearchStart=false
    bRestartTimerOnEachActivation=false
    interval=0.5
    nodeName=Executing...SmartObject
    randomDeviation=0.1
    smartObjectDestinationKeyName=AIEvaluator_SmartObject_Destination
    smartObjectStatusKeyName=AIEvaluator_SmartObject_ExecutionStatus
    evaluationTag {tagName=None}
   }
  }]
}
