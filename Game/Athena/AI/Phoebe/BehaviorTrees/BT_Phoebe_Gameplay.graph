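# Main gameplay behavior tree for Phoebe bots in Athena, choosing actions by priority on BB_Phoebe.
# Blocked movement, launches and DBNO precede the Alive branch, which covers ziplining, diving, gliding and ground actions like trap shooting and danger avoidance.
# When alerted, it runs subtrees for storm, retreat, cover, investigation, combat, loot and patrol.
# AIEvaluator services feed blackboard status keys that gate branches, while tagged injection points admit external behaviors.
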BehaviorTree /Game/Athena/AI/Phoebe/BehaviorTrees/BT_Phoebe_Gameplay
blackboard=&BB_Phoebe
root {
 name=Gameplay
 class=BTComposite_Selector
 children [
  Selector {
   class=BTComposite_Selector
   children [
    Wait {
     class=BTTask_Wait
     decorators [Is Threatened or Lost Target {
       class=BTDecorator_Blackboard
       properties {
        arithmeticOperation=GreaterOrEqual
        intValue=2
        nodeName=Is Threatened or Lost Target
        stringValue=EAlertLevel::LKP
        blackboardKey {selectedKeyName=AIEvaluator_Global_AlertLevel}
       }
      }]
     properties {waitTime {defaultValue=1}}
    }
    Wait {class=BTTask_Wait, properties {waitTime {defaultValue=1}}}
   ]
   decorators [
    Is Movement Blocked? {
     class=BTDecorator_Blackboard
     properties {
      arithmeticOperation=Equal
      intValue=0
      nodeName=Is Movement Blocked?
      stringValue=""
      blackboardKey {selectedKeyName=AIEvaluator_Global_IsMovementBlocked}
     }
    }
    Loop {class=BTDecorator_Loop}
   ]
   properties {
    children [
     {
      childComposite=null
      childTask=&self:BTTask_Wait_7
      decorators [&self:BTDecorator_Blackboard_39]
     }
     {childComposite=null, childTask=&self:BTTask_Wait_4, decorators []}
    ]
    services []
   }
   services []
  }
  Steer Movement {
   class=FortAthenaBTTask_SteerMovement
   children []
   decorators [
    "Behavior Controls = CanEnterCharacterLaunched" {
     class=FortAthenaBTDecorator_BehaviorControls
     properties {
      bInverseCondition=false
      behaviorTreeBranch=CanEnterCharacterLaunched
      flowAbortMode=Both
      nodeName="Behavior Controls = CanEnterCharacterLaunched"
     }
    }
    Is Character being launched? {
     class=BTDecorator_Blackboard
     properties {
      arithmeticOperation=GreaterOrEqual
      intValue=3
      nodeName=Is Character being launched?
      stringValue=EExecutionStatus::ExecutionPending
      blackboardKey {selectedKeyName=AIEvaluator_CharacterLaunched_ExecutionStatus}
     }
    }
   ]
   properties {
    bIgnoreRestartSelf=false
    bSetControlRotation=false
    nodeName=Steer Movement
    steerDirectionKeySelector {bNoneIsAllowedValue=false, selectedKeyName=AIEvaluator_CharacterLaunched_SteerDirection}
   }
   services []
  }
  Run Behavior {
   class=BTTask_RunBehavior
   decorators [Is DBNO {
     class=BTDecorator_Blackboard
     properties {arithmeticOperation=GreaterOrEqual, intValue=3, nodeName=Is DBNO, stringValue="", blackboardKey {selectedKeyName=AIEvaluator_Global_DBNO}}
    }]
   properties {behaviorAsset=&/BT_Phoebe_DBNO, nodeName=Run Behavior}
  }
  Alive {
   class=BTComposite_Selector
   children [
    Injection Point Alive.HighPriority {
     class=FortBTTask_InjectionPoint
     children []
     decorators [Can Do Injected Behavior {
       class=FortBTDecorator_CanDoInjectedBehavior
       properties {
        arithmeticOperation=GreaterOrEqual
        bInverseCondition=false
        basicOperation=Set
        floatValue=0
        flowAbortMode=Both
        intValue=3
        nodeName=Can Do Injected Behavior
        notifyObserver=ResultChange
        stringValue=ExecutionPending
        textOperation=Equal
        blackboardKey {bNoneIsAllowedValue=false, selectedKeyName=AIEvaluator_Injection_Alive_HighPriority}
       }
      }]
     properties {
      bIgnoreRestartSelf=false
      nodeName=Injection Point Alive.HighPriority
      injectedAction {baseClass=&Object, defaultValue=null, key=Injection_Alive_HighPriority_ActionObject}
      injectionTag {tagName=Behavior.Injection.Alive.HighPriority}
     }
     services []
    }
    Ziplining {
     class=FortAthenaBTTask_Zipline
     children []
     decorators [Is on Zipline {
       class=BTDecorator_Blackboard
       properties {
        arithmeticOperation=GreaterOrEqual
        intValue=4
        nodeName=Is on Zipline
        stringValue=EExecutionStatus::ExecutionAllowed
        blackboardKey {selectedKeyName=AIEvaluator_Zipline_UsageExecutionStatus}
       }
      }]
     properties {
      bIgnoreRestartSelf=false
      nodeName=Ziplining
      usageExecutionStatusName=AIEvaluator_Zipline_UsageExecutionStatus
      ziplineTargetName=AIEvaluator_Zipline_Target
     }
     services []
    }
    Diving {
     class=FortAthenaBTTask_Dive
     children []
     decorators [Can Dive {
       class=BTDecorator_Blackboard
       properties {
        arithmeticOperation=Equal
        intValue=4
        nodeName=Can Dive
        stringValue=EExecutionStatus::ExecutionAllowed
        blackboardKey {selectedKeyName=AIEvaluator_Dive_ExecutionStatus}
       }
      }]
     properties {
      bIgnoreRestartSelf=false
      diveDestinationKeyName=AIEvaluator_Dive_Destination
      executionStatusKeyName=AIEvaluator_Dive_ExecutionStatus
      nodeName=Diving
     }
     services []
    }
    Gliding {
     class=FortAthenaBTTask_Glide
     children []
     decorators [Can Glide {
       class=BTDecorator_Blackboard
       properties {
        arithmeticOperation=Equal
        intValue=4
        nodeName=Can Glide
        stringValue=EExecutionStatus::ExecutionAllowed
        blackboardKey {selectedKeyName=AIEvaluator_Glide_ExecutionStatus}
       }
      }]
     properties {
      bIgnoreRestartSelf=false
      executionStatusKeyName=AIEvaluator_Glide_ExecutionStatus
      glideDestinationKeyName=AIEvaluator_Glide_Destination
      nodeName=Gliding
     }
     services []
    }
    Injection Point General.HighPriority {
     class=FortBTTask_InjectionPoint
     children []
     decorators [Can Do Injected Behavior {
       class=FortBTDecorator_CanDoInjectedBehavior
       properties {
        arithmeticOperation=GreaterOrEqual
        bInverseCondition=false
        basicOperation=Set
        floatValue=0
        flowAbortMode=Both
        intValue=3
        nodeName=Can Do Injected Behavior
        notifyObserver=ResultChange
        stringValue=ExecutionPending
        textOperation=Equal
        blackboardKey {bNoneIsAllowedValue=false, selectedKeyName=AIEvaluator_Injection_General_HighPriority}
       }
      }]
     properties {
      bIgnoreRestartSelf=false
      nodeName=Injection Point General.HighPriority
      injectedAction {baseClass=&Object, defaultValue=null, key=Injection_General_HighPriority_ActionObject}
      injectionTag {tagName=Behavior.Injection.General.HighPriority}
     }
     services []
    }
    On Ground {
     class=BTComposite_Selector
     children [
      Bot Shoots Trap {
       class=FortAthenaBTTask_ShootTrap
       children []
       decorators [Trap on Path? {
         class=BTDecorator_Blackboard
         properties {
          arithmeticOperation=GreaterOrEqual
          intValue=3
          nodeName=Trap on Path?
          stringValue=EExecutionStatus::ExecutionPending
          blackboardKey {selectedKeyName=AIEvaluator_TrapOnPath_ExecutionStatus}
         }
        }]
       properties {
        bIgnoreRestartSelf=false
        nodeName=Bot Shoots Trap
        trapOnPathKeyName=AIEvaluator_TrapOnPath_ExecutionStatus
        randomDeviation {defaultValue=0, key=None}
        targetActorKey {bNoneIsAllowedValue=false, selectedKeyName=AIEvaluator_TrapOnPath_Actor}
        waitTime {defaultValue=5, key=None}
       }
       services []
      }
      AvoidThreat MoveTo {
       class=FortAthenaBTTask_BotMoveTo
       decorators [Can AvoidThreat? {
         class=BTDecorator_Blackboard
         properties {
          arithmeticOperation=GreaterOrEqual
          intValue=4
          nodeName=Can AvoidThreat?
          stringValue=EExecutionStatus::ExecutionAllowed
          blackboardKey {selectedKeyName=AIEvaluator_AvoidThreat_ExecutionStatus}
         }
        }]
       properties {
        movementResultKeyName=AIEvaluator_AvoidThreat_MovementState
        nodeName=AvoidThreat MoveTo
        acceptableRadius {defaultValue=100}
        allowRandomWobble {defaultValue=true}
        bAllowStrafe {defaultValue=true}
        blackboardKey {selectedKeyName=AIEvaluator_AvoidThreat_Destination}
        observedBlackboardValueTolerance {defaultValue=4.75}
       }
      }
      Move away from danger {
       class=FortAthenaBTTask_BotMoveTo
       decorators [Danger detected? {
         class=BTDecorator_Blackboard
         properties {
          arithmeticOperation=GreaterOrEqual
          intValue=3
          nodeName=Danger detected?
          stringValue=EExecutionStatus::ExecutionPending
          blackboardKey {selectedKeyName=AIEvaluator_DangerZoneDetected_ExecutionStatus}
         }
        }]
       properties {
        executionStatusKeyName=AIEvaluator_DangerZoneDetected_ExecutionStatus
        movementResultKeyName=None
        nodeName=Move away from danger
        acceptableRadius {defaultValue=5}
        allowRandomWobble {defaultValue=true}
        bAllowStrafe {defaultValue=true}
        bTrackMovingGoal {defaultValue=false}
        blackboardKey {selectedKeyName=AIEvaluator_DangerZoneDetected_SafeLocation}
        observedBlackboardValueTolerance {defaultValue=4.75}
       }
      }
      AWARE {
       class=BTComposite_Selector
       children [
        Selector {
         class=BTComposite_Selector
         children [
          Run Storm Behavior {
           class=BTTask_RunBehavior
           decorators [Escape Storm {
             class=BTDecorator_Blackboard
             properties {
              arithmeticOperation=GreaterOrEqual
              intValue=3
              nodeName=Escape Storm
              stringValue=EExecutionStatus::ExecutionPending
              blackboardKey {selectedKeyName=AIEvaluator_Storm_ExecutionStatus}
             }
            }]
           properties {behaviorAsset=&/BT_Phoebe_Storm, nodeName=Run Storm Behavior}
          }
          Run Retreat Behavior {
           class=BTTask_RunBehavior
           decorators [Can Retreat {
             class=BTDecorator_Blackboard
             properties {
              arithmeticOperation=GreaterOrEqual
              intValue=3
              nodeName=Can Retreat
              stringValue=EExecutionStatus::ExecutionPending
              blackboardKey {selectedKeyName=AIEvaluator_Retreat_ExecutionStatus}
             }
            }]
           properties {behaviorAsset=&/BT_Phoebe_Retreat, nodeName=Run Retreat Behavior}
          }
          Run Bunker Behavior {
           class=BTTask_RunBehavior
           decorators [Should Bunker {
             class=BTDecorator_Blackboard
             properties {
              arithmeticOperation=GreaterOrEqual
              intValue=4
              nodeName=Should Bunker
              stringValue=EExecutionStatus::ExecutionAllowed
              blackboardKey {selectedKeyName=AIEvaluator_Bunker_ExecutionStatus}
             }
            }]
           properties {behaviorAsset=&/BT_Phoebe_Bunker, nodeName=Run Bunker Behavior}
          }
          Run Take Cover Behavior {
           class=BTTask_RunBehavior
           decorators [Should Take Cover {
             class=BTDecorator_Blackboard
             properties {
              arithmeticOperation=GreaterOrEqual
              intValue=3
              nodeName=Should Take Cover
              stringValue=EExecutionStatus::ExecutionPending
              blackboardKey {selectedKeyName=AIEvaluator_TakeCover_ExecutionStatus}
             }
            }]
           properties {behaviorAsset=&/BT_Phoebe_TakeCover, nodeName=Run Take Cover Behavior}
          }
         ]
         decorators []
         properties {
          children [
           {
            childComposite=null
            childTask=&self:BTTask_RunBehavior_15
            decorators [&self:BTDecorator_Blackboard_3]
           }
           {
            childComposite=null
            childTask=&self:BTTask_RunBehavior_4
            decorators [&self:BTDecorator_Blackboard_10]
           }
           {
            childComposite=null
            childTask=&self:BTTask_RunBehavior_6
            decorators [&self:BTDecorator_Blackboard_26]
           }
           {
            childComposite=null
            childTask=&self:BTTask_RunBehavior_9
            decorators [&self:BTDecorator_Blackboard_36]
           }
          ]
          services [
           &self:FortAthenaBTService_AIEvaluator_112
           &self:FortAthenaBTService_AIEvaluator_113
           &self:FortAthenaBTService_AIEvaluator_114
           &self:FortAthenaBTService_AIEvaluator_8
          ]
         }
         services [
          Evaluating...Melee Attack {
           class=FortAthenaBTService_AIEvaluator
           properties {
            bCallTickOnSearchStart=true
            interval=1
            nodeName=Evaluating...Melee Attack
            aIEvaluatorClass {defaultValue=&FortAthenaAIBotEvaluator_MeleeAttack}
           }
          }
          Evaluating...Throwable Attack {
           class=FortAthenaBTService_AIEvaluator
           properties {
            interval=0.2
            nodeName=Evaluating...Throwable Attack
            randomDeviation=0.05
            aIEvaluatorClass {defaultValue=/Script/FortniteAIServer.FortAthenaAIBotEvaluator_ThrowableAttack}
           }
          }
          Evaluating...Range Attack {
           class=FortAthenaBTService_AIEvaluator
           properties {
            bCallTickOnSearchStart=true
            interval=0.2
            nodeName=Evaluating...Range Attack
            randomDeviation=0.05
            aIEvaluatorClass {defaultValue=/Game/Athena/AI/Phoebe/Evaluators/Phoebe_Evaluator_RangeAttack.Phoebe_Evaluator_RangeAttack_C}
           }
          }
          Evaluating...Select Weapon {
           class=FortAthenaBTService_AIEvaluator
           properties {
            interval=0.5
            nodeName=Evaluating...Select Weapon
            randomDeviation=0.05
            aIEvaluatorClass {defaultValue=&FortAthenaAIBotEvaluator_WeaponSelection}
           }
          }
         ]
        }
        Injection Point Threatened.HighPriority {
         class=FortBTTask_InjectionPoint
         children []
         decorators [Can Do Injected Behavior {
           class=FortBTDecorator_CanDoInjectedBehavior
           properties {
            arithmeticOperation=GreaterOrEqual
            bInverseCondition=false
            basicOperation=Set
            floatValue=0
            flowAbortMode=Both
            intValue=3
            nodeName=Can Do Injected Behavior
            notifyObserver=ResultChange
            stringValue=ExecutionPending
            textOperation=Equal
            blackboardKey {bNoneIsAllowedValue=false, selectedKeyName=AIEvaluator_Injection_Threatened_HighPriority}
           }
          }]
         properties {
          bIgnoreRestartSelf=false
          nodeName=Injection Point Threatened.HighPriority
          injectedAction {baseClass=&Object, defaultValue=null, key=Injection_Threatened_ActionObject}
          injectionTag {tagName=Behavior.Injection.Threatened.HighPriority}
         }
         services []
        }
        Selector {
         class=BTComposite_Selector
         children [
          Selector {
           class=BTComposite_Selector
           children [
            Use Heal Item {
             class=FortAthenaBTTask_UseItem
             decorators [
              Can Heal {
               class=BTDecorator_Blackboard
               properties {
                arithmeticOperation=GreaterOrEqual
                intValue=3
                nodeName=Can Heal
                stringValue=EExecutionStatus::ExecutionPending
                blackboardKey {selectedKeyName=AIEvaluator_Healing_ExecutionStatus}
               }
              }
              Distance to cover {
               class=FortBTDecorator_DistanceBetween
               properties {
                flowAbortMode=None
                nodeName=Distance to cover
                operator=Less
                blackboardKeyB {selectedKeyName=AIEvaluator_Retreat_Destination}
                specifiedDistance {defaultValue=250}
               }
              }
             ]
             properties {
              actionObjectKeyName=AIEvaluator_Healing_Object
              executionStatusKeyName=AIEvaluator_Healing_ExecutionStatus
              nodeName=Use Heal Item
              finishAfterItemUse {defaultValue=false}
              inputHoldDuration {key=None}
              maximumWaitTimeBetweenUses {defaultValue=1}
              resetActionObjectKey {defaultValue=true}
             }
            }
            Run Investigation {
             class=BTTask_RunBehavior
             decorators []
             properties {behaviorAsset=&/BT_Phoebe_Investigate, nodeName=Run Investigation}
            }
           ]
           decorators [Should Investigate {
             class=BTDecorator_Blackboard
             properties {
              arithmeticOperation=Equal
              intValue=4
              nodeName=Should Investigate
              stringValue=EExecutionStatus::ExecutionAllowed
              blackboardKey {selectedKeyName=AIEvaluator_Investigate_ExecutionStatus}
             }
            }]
           properties {
            children [
             {
              childComposite=null
              childTask=&self:FortAthenaBTTask_UseItem_2
              decorators [
               &self:BTDecorator_Blackboard_50
               &self:FortBTDecorator_DistanceBetween_3
              ]
             }
             {childComposite=null, childTask=&self:BTTask_RunBehavior_16, decorators []}
            ]
            services [
             &self:FortAthenaBTService_AIEvaluator_110
             &self:FortAthenaBTService_AIEvaluator_111
            ]
           }
           services [
            Evaluating... Heal {
             class=FortAthenaBTService_AIEvaluator
             properties {
              interval=1000000000
              nodeName=Evaluating... Heal
              randomDeviation=0
              aIEvaluatorClass {defaultValue=/Game/Athena/AI/Phoebe/Evaluators/Phoebe_Evaluator_Healing.Phoebe_Evaluator_Healing_C}
             }
            }
            Evaluate...Sprinting {
             class=FortAthenaBTService_AIEvaluator
             properties {
              interval=0.2
              nodeName=Evaluate...Sprinting
              aIEvaluatorClass {defaultValue=/Game/Athena/AI/Phoebe/Evaluators/Phoebe_Evaluator_Sprinting.Phoebe_Evaluator_Sprinting_C}
             }
            }
           ]
          }
          LookAtThreat {
           class=BTComposite_Selector
           children [
            MoveDirectlyToward {
             class=BTTask_MoveDirectlyToward
             children []
             decorators [Can Continue Movement {
               class=BTDecorator_Blackboard
               properties {
                arithmeticOperation=GreaterOrEqual
                intValue=2
                nodeName=Can Continue Movement
                stringValue=""
                blackboardKey {selectedKeyName=AIEvaluator_Observe_Destination}
               }
              }]
             properties {
              bIgnoreRestartSelf=false
              nodeName=MoveDirectlyToward
              acceptableRadius {defaultValue=5, key=None}
              bAllowPartialPath {defaultValue=true, key=None}
              bAllowStrafe {defaultValue=true, key=None}
              bProjectGoalLocation {defaultValue=true, key=None}
              bReachTestIncludesAgentRadius {defaultValue=false, key=None}
              bReachTestIncludesGoalRadius {defaultValue=false, key=None}
              bRequireNavigableEndLocation {defaultValue=true, key=None}
              bStartFromPreviousPath {defaultValue=false, key=None}
              bTrackMovingGoal {defaultValue=true, key=None}
              blackboardKey {bNoneIsAllowedValue=false, selectedKeyName=AIEvaluator_Observe_Destination}
              filterClass {baseClass=&NavigationQueryFilter, defaultValue=null, key=None}
              observedBlackboardValueTolerance {defaultValue=4.75, key=None}
             }
             services []
            }
            MoveDirectlyToward {
             class=BTTask_MoveDirectlyToward
             children []
             decorators [Can Dodge {
               class=BTDecorator_Blackboard
               properties {
                arithmeticOperation=GreaterOrEqual
                intValue=2
                nodeName=Can Dodge
                stringValue=EExecutionStatus::ExecutionSuccess
                blackboardKey {selectedKeyName=AIEvaluator_Dodge_ExecutionStatus}
               }
              }]
             properties {
              bIgnoreRestartSelf=false
              nodeName=MoveDirectlyToward
              acceptableRadius {defaultValue=5, key=None}
              bAllowPartialPath {defaultValue=true, key=None}
              bAllowStrafe {defaultValue=true, key=None}
              bProjectGoalLocation {defaultValue=true, key=None}
              bReachTestIncludesAgentRadius {defaultValue=false, key=None}
              bReachTestIncludesGoalRadius {defaultValue=false, key=None}
              bRequireNavigableEndLocation {defaultValue=true, key=None}
              bStartFromPreviousPath {defaultValue=false, key=None}
              bTrackMovingGoal {defaultValue=true, key=None}
              blackboardKey {bNoneIsAllowedValue=false, selectedKeyName=AIEvaluator_Dodge_Destination}
              filterClass {baseClass=&NavigationQueryFilter, defaultValue=null, key=None}
              observedBlackboardValueTolerance {defaultValue=4.75, key=None}
             }
             services []
            }
            Wait {class=BTTask_Wait, properties {randomDeviation {defaultValue=0.25}, waitTime {defaultValue=1}}}
           ]
           decorators [
            Is Suspicious {
             class=BTDecorator_Blackboard
             properties {
              arithmeticOperation=Equal
              intValue=1
              nodeName=Is Suspicious
              stringValue=EAlertLevel::Alerted
              blackboardKey {selectedKeyName=AIEvaluator_Global_AlertLevel}
             }
            }
            Should Observe {
             class=BTDecorator_Blackboard
             properties {
              arithmeticOperation=GreaterOrEqual
              intValue=3
              nodeName=Should Observe
              stringValue=EExecutionStatus::ExecutionPending
              blackboardKey {selectedKeyName=AIEvaluator_Observe_ExecutionStatus}
             }
            }
            Loop {class=BTDecorator_Loop, properties {infiniteLoopTimeoutTime {defaultValue=4}}}
           ]
           properties {
            bApplyDecoratorScope=true
            nodeName=LookAtThreat
            children [
             {
              childComposite=null
              childTask=&self:BTTask_MoveDirectlyToward_4
              decorators [&self:BTDecorator_Blackboard_38]
             }
             {
              childComposite=null
              childTask=&self:BTTask_MoveDirectlyToward_5
              decorators [&self:BTDecorator_Blackboard_40]
             }
             {childComposite=null, childTask=&self:BTTask_Wait_8, decorators []}
            ]
            services [
             &self:FortAthenaBTService_AIEvaluator_106
             &self:FortAthenaBTService_AIEvaluator_107
             &self:FortAthenaBTService_AIEvaluator_108
             &self:FortAthenaBTService_AIEvaluator_109
            ]
           }
           services [
            Focus on target {
             class=FortAthenaBTService_AIEvaluator
             properties {interval=0.5, nodeName=Focus on target, aIEvaluatorClass {defaultValue=&FortAthenaAIBotEvaluator_HandleFocusing}}
            }
            Evaluating... Evasive Maneuvers {
             class=FortAthenaBTService_AIEvaluator
             properties {
              interval=0.5
              nodeName=Evaluating... Evasive Maneuvers
              aIEvaluatorClass {defaultValue=/Game/Athena/AI/Phoebe/Evaluators/Phoebe_EvasiveManeuvers.Phoebe_EvasiveManeuvers_C}
             }
            }
            Evaluating...AimDownSight {
             class=FortAthenaBTService_AIEvaluator
             properties {
              interval=0.5
              nodeName=Evaluating...AimDownSight
              aIEvaluatorClass {defaultValue=/Script/FortniteAIServer.FortAthenaAIBotEvaluator_AimDownSight}
             }
            }
            Evaluate...Sprinting {
             class=FortAthenaBTService_AIEvaluator
             properties {
              interval=0.2
              nodeName=Evaluate...Sprinting
              aIEvaluatorClass {defaultValue=&Phoebe_Evaluator_Sprinting_C}
             }
            }
           ]
          }
          Sequence {
           class=BTComposite_Sequence
           children [
            Wait {
             class=BTTask_Wait
             decorators [
              Force Success {class=BTDecorator_ForceSuccess}
              Blackboard Based Condition {
               class=BTDecorator_Blackboard
               properties {
                arithmeticOperation=Equal
                flowAbortMode=Self
                intValue=0
                nodeName=Blackboard Based Condition
                stringValue=""
                blackboardKey {selectedKeyName=AIEvaluator_TakeCover_Destination}
               }
              }
             ]
             properties {waitTime {defaultValue=2}}
            }
            Move To Range {
             class=BTTask_RunBehavior
             decorators [Loop {class=BTDecorator_Loop}]
             properties {behaviorAsset=&BT_Phoebe_MoveToRange, nodeName=Move To Range}
            }
           ]
           decorators [
            Is Threatened or Lost Target {
             class=BTDecorator_Blackboard
             properties {
              arithmeticOperation=GreaterOrEqual
              intValue=2
              nodeName=Is Threatened or Lost Target
              stringValue=EAlertLevel::LKP
              blackboardKey {selectedKeyName=AIEvaluator_Global_AlertLevel}
             }
            }
            Blackboard Based Condition {
             class=BTDecorator_Blackboard
             properties {
              arithmeticOperation=GreaterOrEqual
              intValue=3
              nodeName=Blackboard Based Condition
              stringValue=EExecutionStatus::ExecutionPending
              blackboardKey {selectedKeyName=AIEvaluator_RangeAttack_ExecutionStatus}
             }
            }
            Blackboard Based Condition {
             class=BTDecorator_Blackboard
             properties {
              arithmeticOperation=GreaterOrEqual
              intValue=3
              nodeName=Blackboard Based Condition
              stringValue=EExecutionStatus::ExecutionPending
              blackboardKey {selectedKeyName=AIEvaluator_ThrowableAttack_ExecutionStatus}
             }
            }
           ]
           properties {
            children [
             {
              childTask=&self:BTTask_Wait_2
              decorators [
               &self:BTDecorator_ForceSuccess_0
               &self:BTDecorator_Blackboard_24
              ]
             }
             {
              childTask=&self:BTTask_RunBehavior_17
              decorators [&self:BTDecorator_Loop_5]
             }
            ]
            services [
             &self:FortAthenaBTService_AIEvaluator_79
             &self:FortAthenaBTService_AIEvaluator_80
             &self:FortAthenaBTService_AIEvaluator_81
             &self:FortAthenaBTService_AIEvaluator_13
            ]
           }
           services [
            "Evaluating...Evasive Maneuvers (Range Fight)" {
             class=FortAthenaBTService_AIEvaluator
             properties {
              interval=0.5
              nodeName="Evaluating...Evasive Maneuvers (Range Fight)"
              aIEvaluatorClass {defaultValue=/Game/Athena/AI/Phoebe/Evaluators/Phoebe_EvasiveManeuvers_RangeFight.Phoebe_EvasiveManeuvers_RangeFight_C}
             }
            }
            Focus on target {
             class=FortAthenaBTService_AIEvaluator
             properties {interval=0.5, nodeName=Focus on target, aIEvaluatorClass {defaultValue=&FortAthenaAIBotEvaluator_HandleFocusing}}
            }
            Evaluating...AimDownSight {
             class=FortAthenaBTService_AIEvaluator
             properties {
              interval=0.5
              nodeName=Evaluating...AimDownSight
              aIEvaluatorClass {defaultValue=&FortAthenaAIBotEvaluator_AimDownSight}
             }
            }
            Evaluate...Sprinting {
             class=FortAthenaBTService_AIEvaluator
             properties {
              interval=0.2
              nodeName=Evaluate...Sprinting
              aIEvaluatorClass {defaultValue=&Phoebe_Evaluator_Sprinting_C}
             }
            }
           ]
          }
          Selector {
           class=BTComposite_Selector
           children [Move To Melee {
             class=BTTask_RunBehavior
             decorators [Loop {class=BTDecorator_Loop}]
             properties {behaviorAsset=&/BT_Phoebe_MoveToMelee, nodeName=Move To Melee}
            }]
           decorators [
            Is Threatened or Lost Target {
             class=BTDecorator_Blackboard
             properties {
              arithmeticOperation=GreaterOrEqual
              intValue=2
              nodeName=Is Threatened or Lost Target
              stringValue=EAlertLevel::LKP
              blackboardKey {selectedKeyName=AIEvaluator_Global_AlertLevel}
             }
            }
            Can Do MeleeAttack {
             class=BTDecorator_Blackboard
             properties {
              arithmeticOperation=GreaterOrEqual
              intValue=3
              nodeName=Can Do MeleeAttack
              stringValue=EExecutionStatus::ExecutionPending
              blackboardKey {selectedKeyName=AIEvaluator_MeleeAttack_ExecutionStatus}
             }
            }
           ]
           properties {
            children [{
              childComposite=null
              childTask=&self:BTTask_RunBehavior_2
              decorators [&self:BTDecorator_Loop_1]
             }]
            services [
             &self:FortAthenaBTService_AIEvaluator_20
             &self:FortAthenaBTService_AIEvaluator_24
             &self:FortAthenaBTService_AIEvaluator_12
            ]
           }
           services [
            Focus on target {
             class=FortAthenaBTService_AIEvaluator
             properties {interval=0.5, nodeName=Focus on target, aIEvaluatorClass {defaultValue=&FortAthenaAIBotEvaluator_HandleFocusing}}
            }
            "Evaluating...Evasive Maneuvers (Melee Fight)" {
             class=FortAthenaBTService_AIEvaluator
             properties {
              interval=0.5
              nodeName="Evaluating...Evasive Maneuvers (Melee Fight)"
              aIEvaluatorClass {defaultValue=/Game/Athena/AI/Phoebe/Evaluators/Phoebe_EvasiveManeuvers_MeleeFight.Phoebe_EvasiveManeuvers_MeleeFight_C}
             }
            }
            Evaluate...Sprinting {
             class=FortAthenaBTService_AIEvaluator
             properties {
              interval=0.2
              nodeName=Evaluate...Sprinting
              aIEvaluatorClass {defaultValue=&Phoebe_Evaluator_Sprinting_C}
             }
            }
           ]
          }
          Run Loot Behavior {
           class=BTTask_RunBehavior
           decorators [Loot {
             class=BTDecorator_Blackboard
             properties {
              arithmeticOperation=GreaterOrEqual
              intValue=2
              nodeName=Loot
              stringValue=EExecutionStatus::ExecutionSuccess
              blackboardKey {selectedKeyName=AIEvaluator_Loot_ExecutionStatus}
             }
            }]
           properties {behaviorAsset=&/BT_Phoebe_Loot, nodeName=Run Loot Behavior}
          }
          Fallback Behavior {
           class=BTComposite_Selector
           children [
            Run PatrolAround Behavior {
             class=BTTask_RunBehavior
             decorators [Patrol Around {
               class=BTDecorator_Blackboard
               properties {
                arithmeticOperation=GreaterOrEqual
                intValue=2
                nodeName=Patrol Around
                stringValue=EExecutionStatus::ExecutionSuccess
                blackboardKey {selectedKeyName=AIEvaluator_Patrol_ExecutionStatus}
               }
              }]
             properties {behaviorAsset=&/BT_Phoebe_PatrolAround, nodeName=Run PatrolAround Behavior}
            }
            Wait {class=BTTask_Wait, properties {waitTime {defaultValue=2}}}
           ]
           decorators []
           properties {
            nodeName=Fallback Behavior
            children [
             {
              childComposite=null
              childTask=&self:BTTask_RunBehavior_13
              decorators [&self:BTDecorator_Blackboard_2]
             }
             {childComposite=null, childTask=&self:BTTask_Wait_1, decorators []}
            ]
            services [
             &self:FortAthenaBTService_AIEvaluator_23
             &self:FortAthenaBTService_AIEvaluator_30
            ]
           }
           services [
            Evaluating ... Patrol Around {
             class=FortAthenaBTService_AIEvaluator
             properties {
              bCallTickOnSearchStart=true
              interval=2.5
              nodeName=Evaluating ... Patrol Around
              randomDeviation=0.5
              aIEvaluatorClass {defaultValue=/Script/FortniteAIServer.FortAthenaAIBotEvaluator_PatrolAround}
             }
            }
            Focus Target Ignore Threat {
             class=FortAthenaBTService_AIEvaluator
             properties {
              interval=0.5
              nodeName=Focus Target Ignore Threat
              aIEvaluatorClass {defaultValue=/Game/Athena/AI/Phoebe/Evaluators/Phoebe_HandleFocusing_IgnoreThreatAfterTimer.Phoebe_HandleFocusing_IgnoreThreatAfterTimer_C}
             }
            }
           ]
          }
         ]
         decorators []
         properties {
          children [
           {
            childComposite=&self:BTComposite_Selector_24
            childTask=null
            decorators [&self:BTDecorator_Blackboard_53]
           }
           {
            childComposite=&self:BTComposite_Selector_22
            childTask=null
            decorators [
             &self:BTDecorator_Blackboard_44
             &self:BTDecorator_Blackboard_45
             &self:BTDecorator_Loop_8
            ]
           }
           {
            childComposite=&self:BTComposite_Sequence_1
            childTask=null
            decoratorOps [{number=2, operation=And}, {number=0}, {number=2, operation=Or}, {number=1}, {number=2}]
            decorators [
             &self:BTDecorator_Blackboard_55
             &self:BTDecorator_Blackboard_59
             &self:BTDecorator_Blackboard_58
            ]
           }
           {
            childComposite=&self:BTComposite_Selector_2
            childTask=null
            decorators [
             &self:BTDecorator_Blackboard_31
             &self:BTDecorator_Blackboard_5
            ]
           }
           {
            childComposite=null
            childTask=&self:BTTask_RunBehavior_0
            decorators [&self:BTDecorator_Blackboard_7]
           }
           {childComposite=&self:BTComposite_Selector_13, childTask=null, decorators []}
          ]
          services [
           &self:FortAthenaBTService_AIEvaluator_44
           &self:FortAthenaBTService_AIEvaluator_48
           &self:FortAthenaBTService_AIEvaluator_49
           &self:FortAthenaBTService_AIEvaluator_7
          ]
         }
         services [
          Evaluating...Melee Attack {
           class=FortAthenaBTService_AIEvaluator
           properties {
            bCallTickOnSearchStart=true
            interval=1
            nodeName=Evaluating...Melee Attack
            aIEvaluatorClass {defaultValue=&FortAthenaAIBotEvaluator_MeleeAttack}
           }
          }
          Evaluating...Throwable Attack {
           class=FortAthenaBTService_AIEvaluator
           properties {
            interval=0.2
            nodeName=Evaluating...Throwable Attack
            randomDeviation=0.05
            aIEvaluatorClass {defaultValue=&FortAthenaAIBotEvaluator_ThrowableAttack}
           }
          }
          Evaluating...Range Attack {
           class=FortAthenaBTService_AIEvaluator
           properties {
            bCallTickOnSearchStart=true
            interval=0.2
            nodeName=Evaluating...Range Attack
            randomDeviation=0.05
            aIEvaluatorClass {defaultValue=&Phoebe_Evaluator_RangeAttack_C}
           }
          }
          Evaluating...Select Weapon {
           class=FortAthenaBTService_AIEvaluator
           properties {
            interval=0.5
            nodeName=Evaluating...Select Weapon
            randomDeviation=0.05
            aIEvaluatorClass {defaultValue=&FortAthenaAIBotEvaluator_WeaponSelection}
           }
          }
         ]
        }
       ]
       decorators ["IsThreatened, LostTarget or Suspicious" {
         class=BTDecorator_Blackboard
         properties {
          arithmeticOperation=GreaterOrEqual
          intValue=1
          nodeName="IsThreatened, LostTarget or Suspicious"
          stringValue=EAlertLevel::Alerted
          blackboardKey {selectedKeyName=AIEvaluator_Global_AlertLevel}
         }
        }]
       properties {
        nodeName=AWARE
        children [
         {childComposite=&self:BTComposite_Selector_20, childTask=null, decorators []}
         {
          childComposite=null
          childTask=&self:FortBTTask_InjectedAction_0
          decorators [&self:FortBTDecorator_CanDoInjectedBehavior_5]
         }
         {childComposite=&self:BTComposite_Selector_6, childTask=null, decorators []}
        ]
        services [
         &self:FortAthenaBTService_ManageWeapon_2
         &self:FortAthenaBTService_AIEvaluator_25
         &self:FortAthenaBTService_AIEvaluator_32
         &self:FortAthenaBTService_AIEvaluator_37
         &self:FortAthenaBTService_AIEvaluator_61
         &self:FortBTService_AIEvaluatorsInjector_0
        ]
       }
       services [
        Executing...Weapons Actions {class=FortAthenaBTService_ManageWeapon, properties {interval=1, randomDeviation=0.3}}
        Evaluating...Investigate Treat {
         class=FortAthenaBTService_AIEvaluator
         properties {
          bCallTickOnSearchStart=true
          interval=0.5
          nodeName=Evaluating...Investigate Treat
          aIEvaluatorClass {defaultValue=/Game/Athena/AI/Phoebe/Evaluators/Phoebe_Evaluator_Investigate.Phoebe_Evaluator_Investigate_C}
         }
        }
        Evaluating... Bunker {
         class=FortAthenaBTService_AIEvaluator
         properties {
          interval=0.5
          nodeName=Evaluating... Bunker
          aIEvaluatorClass {defaultValue=/Game/Athena/AI/Phoebe/Evaluators/Phoebe_Evaluator_Bunker.Phoebe_Evaluator_Bunker_C}
         }
        }
        Evaluating... Take Cover {
         class=FortAthenaBTService_AIEvaluator
         properties {interval=0.5, nodeName=Evaluating... Take Cover, aIEvaluatorClass {defaultValue=/Script/FortniteAIServer.FortAthenaAIBotEvaluator_TakeCover}}
        }
        Evaluating...Observe Treat {
         class=FortAthenaBTService_AIEvaluator
         properties {
          bCallTickOnSearchStart=true
          interval=0.5
          nodeName=Evaluating...Observe Treat
          aIEvaluatorClass {defaultValue=/Game/Athena/AI/Phoebe/Evaluators/Phoebe_Evaluator_Observe.Phoebe_Evaluator_Observe_C}
         }
        }
        Behavior Injector {
         class=FortBTService_AIEvaluatorsInjector
         properties {
          bCallTickOnSearchStart=false
          bRestartTimerOnEachActivation=false
          interval=0.5
          nodeName=Behavior Injector
          randomDeviation=0.1
          injectionTagsKeys [{
            actionInterfaceObjectKeyName=Injection_Threatened_ActionObject
            injectionKeyName=AIEvaluator_Injection_Threatened_HighPriority
            injectionTag {tagName=Behavior.Injection.Threatened.HighPriority}
           }]
         }
        }
       ]
      }
      Not attacking threat {
       class=BTComposite_Selector
       children [
        Run Behavior {
         class=BTTask_RunBehavior
         decorators [Revive {
           class=BTDecorator_Blackboard
           properties {
            arithmeticOperation=GreaterOrEqual
            intValue=3
            nodeName=Revive
            stringValue=EExecutionStatus::ExecutionPending
            blackboardKey {selectedKeyName=AIEvaluator_Revive_ExecutionStatus}
           }
          }]
         properties {behaviorAsset=&/BT_Phoebe_Revive, nodeName=Run Behavior}
        }
        PlayEmote {
         class=FortAthenaBTTask_PlayEmote
         children []
         decorators [Can DanceOnKill? {
           class=BTDecorator_Blackboard
           properties {
            arithmeticOperation=GreaterOrEqual
            flowAbortMode=LowerPriority
            intValue=3
            nodeName=Can DanceOnKill?
            stringValue=EExecutionStatus::ExecutionPending
            blackboardKey {selectedKeyName=AIEvaluator_PlayEmote_ExecutionStatus}
           }
          }]
         properties {bIgnoreRestartSelf=false, nodeName=PlayEmote, playEmoteExecutionStatusKeyName=AIEvaluator_PlayEmote_ExecutionStatus, bespokeEmotes []}
         services []
        }
        UNAWARE {
         class=BTComposite_Selector
         children [
          Run Storm Behavior {
           class=BTTask_RunBehavior
           decorators [Escape Storm {
             class=BTDecorator_Blackboard
             properties {
              arithmeticOperation=GreaterOrEqual
              intValue=3
              nodeName=Escape Storm
              stringValue=EExecutionStatus::ExecutionPending
              blackboardKey {selectedKeyName=AIEvaluator_Storm_ExecutionStatus}
             }
            }]
           properties {behaviorAsset=&BT_Phoebe_Storm, nodeName=Run Storm Behavior}
          }
          Injection Point Athena.Branch.Unaware.HighPriority {
           class=FortBTTask_InjectionPoint
           children []
           decorators [Can Do Injected Behavior {
             class=FortBTDecorator_CanDoInjectedBehavior
             properties {
              arithmeticOperation=GreaterOrEqual
              bInverseCondition=false
              basicOperation=Set
              floatValue=0
              flowAbortMode=Both
              intValue=3
              nodeName=Can Do Injected Behavior
              notifyObserver=ResultChange
              stringValue=ExecutionPending
              textOperation=Equal
              blackboardKey {bNoneIsAllowedValue=false, selectedKeyName=AIEvaluator_Injection_Unaware_HighPriority}
             }
            }]
           properties {
            bIgnoreRestartSelf=false
            nodeName=Injection Point Athena.Branch.Unaware.HighPriority
            injectedAction {baseClass=&Object, defaultValue=null, key=Injection_Unaware_ActionObject}
            injectionTag {tagName=Behavior.Injection.Athena.Branch.Unaware.HighPriority}
           }
           services []
          }
          UseSmartObject {
           class=BTTask_RunBehavior
           decorators []
           properties {behaviorAsset=&/BT_Phoebe_UseSmartObject, nodeName=UseSmartObject}
          }
          Selector {
           class=BTComposite_Selector
           children [
            Move to Next POI Selector {
             class=BTComposite_Selector
             children [
              Run Unstuck Behavior {
               class=BTTask_RunBehavior
               decorators []
               properties {behaviorAsset=&BT_Phoebe_UnstuckNew, nodeName=Run Unstuck Behavior}
              }
              Selector {
               class=BTComposite_Selector
               children [
                Run Zipline Behavior {
                 class=BTTask_RunBehavior
                 decorators [
                  Blackboard Based Condition {
                   class=BTDecorator_Blackboard
                   properties {
                    arithmeticOperation=GreaterOrEqual
                    intValue=4
                    nodeName=Blackboard Based Condition
                    stringValue=EExecutionStatus::ExecutionAllowed
                    blackboardKey {selectedKeyName=AIEvaluator_Zipline_MoveExecutionStatus}
                   }
                  }
                  Blackboard Based Condition {
                   class=BTDecorator_Blackboard
                   properties {
                    arithmeticOperation=GreaterOrEqual
                    intValue=4
                    nodeName=Blackboard Based Condition
                    stringValue=EExecutionStatus::ExecutionAllowed
                    blackboardKey {selectedKeyName=AIEvaluator_Zipline_UsageExecutionStatus}
                   }
                  }
                 ]
                 properties {behaviorAsset=&BT_Phoebe_Zipline, nodeName=Run Zipline Behavior}
                }
                Move To Next Dynamic POI {
                 class=FortAthenaBTTask_BotMoveTo
                 decorators []
                 properties {
                  executionStatusKeyName=AIEvaluator_DynamicPOINavigation_ExecutionStatus
                  movementResultKeyName=AIEvaluator_DynamicPOINavigation_MovementState
                  nodeName=Move To Next Dynamic POI
                  acceptableRadius {defaultValue=50}
                  allowRandomWobble {defaultValue=true}
                  bAllowStrafe {defaultValue=true}
                  blackboardKey {selectedKeyName=AIEvaluator_DynamicPOINavigation_Location}
                  observedBlackboardValueTolerance {defaultValue=200}
                 }
                }
               ]
               decorators []
               properties {
                children [
                 {
                  childComposite=null
                  childTask=&self:BTTask_RunBehavior_11
                  decoratorOps [{number=2, operation=Or}, {number=0}, {number=1}]
                  decorators [
                   &self:BTDecorator_Blackboard_51
                   &self:BTDecorator_Blackboard_52
                  ]
                 }
                 {childComposite=null, childTask=&self:FortAthenaBTTask_BotMoveTo_5, decorators []}
                ]
                services [&self:FortAthenaBTService_AIEvaluator_31]
               }
               services [Evaluate...DynamicPOI Zipline {
                 class=FortAthenaBTService_AIEvaluator
                 properties {
                  bCallTickOnSearchStart=true
                  interval=0.5
                  nodeName=Evaluate...DynamicPOI Zipline
                  aIEvaluatorClass {defaultValue=/Game/Athena/AI/Phoebe/Evaluators/Phoebe_Evaluator_Zipline_MoveToNextDynamicPOI.Phoebe_Evaluator_Zipline_MoveToNextDynamicPOI_C}
                 }
                }]
              }
             ]
             decorators [Should move to next Dynamic POI {
               class=BTDecorator_Blackboard
               properties {
                arithmeticOperation=GreaterOrEqual
                intValue=3
                nodeName=Should move to next Dynamic POI
                stringValue=ExecutionPending
                blackboardKey {selectedKeyName=AIEvaluator_DynamicPOINavigation_ExecutionStatus}
               }
              }]
             properties {
              nodeName=Move to Next POI Selector
              children [
               {childComposite=null, childTask=&self:BTTask_RunBehavior_14, decorators []}
               {childComposite=&self:BTComposite_Selector_8, childTask=null, decorators []}
              ]
              services []
             }
             services []
            }
            Run Loot Behavior {
             class=BTTask_RunBehavior
             decorators [Loot {
               class=BTDecorator_Blackboard
               properties {
                arithmeticOperation=GreaterOrEqual
                intValue=2
                nodeName=Loot
                stringValue=EExecutionStatus::ExecutionSuccess
                blackboardKey {selectedKeyName=AIEvaluator_Loot_ExecutionStatus}
               }
              }]
             properties {behaviorAsset=&BT_Phoebe_Loot, nodeName=Run Loot Behavior}
            }
            Fallback Behavior {
             class=BTComposite_Selector
             children [
              Run PatrolAround Behavior {
               class=BTTask_RunBehavior
               decorators [Patrol Around {
                 class=BTDecorator_Blackboard
                 properties {
                  arithmeticOperation=GreaterOrEqual
                  intValue=2
                  nodeName=Patrol Around
                  stringValue=EExecutionStatus::ExecutionSuccess
                  blackboardKey {selectedKeyName=AIEvaluator_Patrol_ExecutionStatus}
                 }
                }]
               properties {behaviorAsset=&BT_Phoebe_PatrolAround, nodeName=Run PatrolAround Behavior}
              }
              Wait {class=BTTask_Wait, properties {waitTime {defaultValue=2}}}
             ]
             decorators []
             properties {
              nodeName=Fallback Behavior
              children [
               {
                childComposite=null
                childTask=&self:BTTask_RunBehavior_10
                decorators [&self:BTDecorator_Blackboard_41]
               }
               {childComposite=null, childTask=&self:BTTask_Wait_0, decorators []}
              ]
              services [&self:FortAthenaBTService_AIEvaluator_3]
             }
             services [Evaluating ... Patrol Around {
               class=FortAthenaBTService_AIEvaluator
               properties {
                bCallTickOnSearchStart=true
                interval=2.5
                nodeName=Evaluating ... Patrol Around
                randomDeviation=0.5
                aIEvaluatorClass {defaultValue=&FortAthenaAIBotEvaluator_PatrolAround}
               }
              }]
            }
           ]
           decorators []
           properties {
            children [
             {
              childComposite=&self:BTComposite_Selector_11
              childTask=null
              decorators [&self:BTDecorator_Blackboard_29]
             }
             {
              childComposite=null
              childTask=&self:BTTask_RunBehavior_18
              decorators [&self:BTDecorator_Blackboard_43]
             }
             {childComposite=&self:BTComposite_Selector_4, childTask=null, decorators []}
            ]
            services [
             &self:FortAthenaBTService_AIEvaluator_46
             &self:FortAthenaBTService_AIEvaluator_4
             &self:FortAthenaBTService_UseItem_0
             &self:FortAthenaBTService_UseItem_4
             &self:FortAthenaBTService_AIEvaluator_51
            ]
           }
           services [
            Focus Scan Around {
             class=FortAthenaBTService_AIEvaluator
             properties {
              interval=0.5
              nodeName=Focus Scan Around
              aIEvaluatorClass {defaultValue=/Game/Athena/AI/Phoebe/Evaluators/Phoebe_HandleFocusing_ScanAround.Phoebe_HandleFocusing_ScanAround_C}
             }
            }
            Evaluating... Use Augmented Movement Items {
             class=FortAthenaBTService_AIEvaluator
             properties {
              interval=0.5
              nodeName=Evaluating... Use Augmented Movement Items
              aIEvaluatorClass {defaultValue=/Game/Athena/AI/Phoebe/Evaluators/Phoebe_Evaluator_UseAugmentMovementItems.Phoebe_Evaluator_UseAugmentMovementItems_C}
             }
            }
            Executing... Use Heal Items {
             class=FortAthenaBTService_UseItem
             properties {
              actionObjectKeyName=AIEvaluator_Healing_Object
              bBlockWeaponActions=true
              bCallTickOnSearchStart=false
              bCanOnlyUseMobileItems=false
              bResetActionObjectKey=true
              bRestartTimerOnEachActivation=false
              bValidateAbility=true
              canMoveWhileUsingItemKeyName=AIEvaluator_Healing_CanUseWhileMoving
              equipRetryInterval=0.5
              executionStatusKeyName=AIEvaluator_Healing_ExecutionStatus
              interval=999
              itemUsageTimeout=15
              maxEquipAttempts=5
              maxWaitTimeBetweenUses=1
              minWaitTimeBetweenUses=0
              nodeName=Executing... Use Heal Items
              randomDeviation=0.1
              useItemDelay=0.25
              bCanUseInputToDeactivateKeyName {defaultValue=false, key=None}
              bUseAlternateMode {defaultValue=false, key=None}
             }
            }
            Executing... Use Movement Items {
             class=FortAthenaBTService_UseItem
             properties {
              actionObjectKeyName=AIEvaluator_MovementItem_Object
              bBlockWeaponActions=true
              bCallTickOnSearchStart=false
              bCanOnlyUseMobileItems=true
              bResetActionObjectKey=true
              bRestartTimerOnEachActivation=false
              bValidateAbility=true
              canMoveWhileUsingItemKeyName=AIEvaluator_MovementItem_CanUseWhilePathFollowing
              equipRetryInterval=0.5
              executionStatusKeyName=AIEvaluator_MovementItem_ExecutionStatus
              interval=999999
              itemUsageTimeout=15
              maxEquipAttempts=5
              maxWaitTimeBetweenUses=1
              minWaitTimeBetweenUses=0
              nodeName=Executing... Use Movement Items
              randomDeviation=0.1
              useItemDelay=0.25
              bCanUseInputToDeactivateKeyName {defaultValue=false, key=AIEvaluator_MovementItem_CanUseInputToDeactivate}
              bUseAlternateMode {defaultValue=false, key=AIEvaluator_MovementItem_UseSecondaryFire}
             }
            }
            Evaluate...Sprinting {
             class=FortAthenaBTService_AIEvaluator
             properties {
              interval=0.2
              nodeName=Evaluate...Sprinting
              aIEvaluatorClass {defaultValue=&Phoebe_Evaluator_Sprinting_C}
             }
            }
           ]
          }
         ]
         decorators [Loop {class=BTDecorator_Loop}]
         properties {
          nodeName=UNAWARE
          children [
           {
            childComposite=null
            childTask=&self:BTTask_RunBehavior_21
            decorators [&self:BTDecorator_Blackboard_46]
           }
           {
            childComposite=null
            childTask=&self:FortBTTask_InjectionPoint_4
            decorators [&self:FortBTDecorator_CanDoInjectedBehavior_6]
           }
           {childComposite=null, childTask=&self:BTTask_RunBehavior_8, decorators []}
           {childComposite=&self:BTComposite_Selector_0, childTask=null, decorators []}
          ]
          services [
           &self:FortAthenaBTService_AIEvaluator_50
           &self:FortAthenaBTService_AIEvaluator_39
           &self:FortAthenaBTService_AIEvaluator_45
           &self:FortAthenaBTService_ManageWeapon_6
           &self:FortAthenaBTService_AIEvaluator_28
           &self:FortAthenaBTService_AIEvaluator_27
           &self:FortBTService_AIEvaluatorsInjector_2
          ]
         }
         services [
          Evaluating...Select Weapon {
           class=FortAthenaBTService_AIEvaluator
           properties {
            interval=0.5
            nodeName=Evaluating...Select Weapon
            randomDeviation=0.05
            aIEvaluatorClass {defaultValue=&FortAthenaAIBotEvaluator_WeaponSelection}
           }
          }
          Evaluating...Range Attack {
           class=FortAthenaBTService_AIEvaluator
           properties {
            interval=0.3
            nodeName=Evaluating...Range Attack
            aIEvaluatorClass {defaultValue=&Phoebe_Evaluator_RangeAttack_C}
           }
          }
          Evaluating...Melee Attack {
           class=FortAthenaBTService_AIEvaluator
           properties {
            interval=0.3
            nodeName=Evaluating...Melee Attack
            aIEvaluatorClass {defaultValue=&FortAthenaAIBotEvaluator_MeleeAttack}
           }
          }
          Executing...Weapons Actions {class=FortAthenaBTService_ManageWeapon, properties {interval=0.2, randomDeviation=0.1}}
          Evaluating... Reload {
           class=FortAthenaBTService_AIEvaluator
           properties {
            interval=0.5
            nodeName=Evaluating... Reload
            aIEvaluatorClass {defaultValue=/Game/Athena/AI/Phoebe/Evaluators/Phoebe_Evaluator_ReloadWeapon_OnlyCurrent.Phoebe_Evaluator_ReloadWeapon_OnlyCurrent_C}
           }
          }
          Evaluating...SmartObject {
           class=FortAthenaBTService_AIEvaluator
           properties {
            interval=2
            nodeName=Evaluating...SmartObject
            aIEvaluatorClass {defaultValue=/Script/FortniteAIServer.FortAthenaAIBotEvaluator_SmartObjects}
           }
          }
          Behavior Injector {
           class=FortBTService_AIEvaluatorsInjector
           properties {
            bCallTickOnSearchStart=false
            bRestartTimerOnEachActivation=false
            interval=0.5
            nodeName=Behavior Injector
            randomDeviation=0.1
            injectionTagsKeys [{
              actionInterfaceObjectKeyName=Injection_Unaware_ActionObject
              injectionKeyName=AIEvaluator_Injection_Unaware_HighPriority
              injectionTag {tagName=Behavior.Injection.Athena.Branch.Unaware.HighPriority}
             }]
           }
          }
         ]
        }
       ]
       decorators []
       properties {
        nodeName=Not attacking threat
        children [
         {
          childComposite=null
          childTask=&self:BTTask_RunBehavior_7
          decorators [&self:BTDecorator_Blackboard_47]
         }
         {
          childComposite=null
          childTask=&self:FortAthenaBTTask_PlayEmote_0
          decorators [&self:BTDecorator_Blackboard_19]
         }
         {
          childComposite=&self:BTComposite_Selector_18
          childTask=null
          decorators [&self:BTDecorator_Loop_2]
         }
        ]
        services [
         &self:FortAthenaBTService_AIEvaluator_54
         &self:FortAthenaBTService_AIEvaluator_36
         &self:FortAthenaBTService_AIEvaluator_33
        ]
       }
       services [
        Evaluation...NextDynamicPOI {
         class=FortAthenaBTService_AIEvaluator
         properties {
          interval=2.5
          nodeName=Evaluation...NextDynamicPOI
          randomDeviation=0.5
          aIEvaluatorClass {defaultValue=/Game/Athena/AI/Phoebe/Evaluators/Phoebe_Evaluator_DynamicNextPOI.Phoebe_Evaluator_DynamicNextPOI_C}
         }
        }
        Evaluating...DanceOnKill {
         class=FortAthenaBTService_AIEvaluator
         properties {
          bCallTickOnSearchStart=true
          interval=0.5
          nodeName=Evaluating...DanceOnKill
          aIEvaluatorClass {defaultValue=/Script/FortniteAIServer.FortAthenaAIBotEvaluator_DanceOnKill}
         }
        }
        Evaluating... Heal {
         class=FortAthenaBTService_AIEvaluator
         properties {
          interval=10
          nodeName=Evaluating... Heal
          randomDeviation=0
          aIEvaluatorClass {defaultValue=&Phoebe_Evaluator_Healing_C}
         }
        }
       ]
      }
     ]
     decorators [Has Landed From Bus {
       class=BTDecorator_HasEverSkydivedFromBusAndLanded_C
       properties {
        bCheckConditionOnlyBlackBoardChanges=false
        bInverseCondition=false
        bShowPropertyDetails=true
        customDescription=""
        flowAbortMode=None
        nodeName=Has Landed From Bus
       }
      }]
     properties {
      nodeName=On Ground
      children [
       {
        childComposite=null
        childTask=&self:FortAthenaBTTask_ShootTrap_1
        decorators [&self:BTDecorator_Blackboard_63]
       }
       {
        childComposite=null
        childTask=&self:FortAthenaBTTask_BotMoveTo_2
        decorators [&self:BTDecorator_Blackboard_65]
       }
       {
        childComposite=null
        childTask=&self:FortAthenaBTTask_BotMoveTo_0
        decorators [&self:BTDecorator_Blackboard_64]
       }
       {
        childComposite=&self:BTComposite_Selector_5
        childTask=null
        decorators [&self:BTDecorator_Blackboard_18]
       }
       {childComposite=&self:BTComposite_Selector_15, childTask=null, decorators []}
      ]
      services [
       &self:FortAthenaBTService_AIEvaluator_56
       &self:FortAthenaBTService_AIEvaluator_58
       &self:FortAthenaBTService_AIEvaluator_71
       &self:FortAthenaBTService_AIEvaluator_72
       &self:FortAthenaBTService_AIEvaluator_73
       &self:FortAthenaBTService_PickUpLoot_1
       &self:FortAthenaBTService_Crouch_2
       &self:FortAthenaBTService_Slide_1
       &self:FortAthenaBTService_Jump_2
       &self:FortAthenaBTService_Clamber_1
       &self:FortAthenaBTService_Sprinting_2
       &self:FortAthenaBTService_BuildConstruction_1
       &self:FortAthenaBTService_Escape_1
       &self:FortAthenaBTService_AIEvaluator_74
       &self:FortAthenaBTService_AIEvaluator_75
       &self:FortAthenaBTService_AIEvaluator_76
       &self:FortAthenaBTService_AIEvaluator_78
       &self:FortAthenaBTService_AIEvaluator_85
       &self:FortAthenaBTService_AIEvaluator_86
       &self:FortAthenaBTService_AIEvaluator_87
       &self:FortAthenaBTService_AIEvaluator_88
       &self:FortAthenaBTService_AIEvaluator_89
      ]
     }
     services [
      Evaluate...Loot {
       class=FortAthenaBTService_AIEvaluator
       properties {
        bCallTickOnSearchStart=true
        interval=0.5
        nodeName=Evaluate...Loot
        aIEvaluatorClass {defaultValue=/Script/FortniteAI.FortAthenaAIBotEvaluator_Loot}
       }
      }
      Evaluate...Revive {
       class=FortAthenaBTService_AIEvaluator
       properties {interval=0.5, nodeName=Evaluate...Revive, aIEvaluatorClass {defaultValue=/Script/FortniteAIServer.FortAthenaAIBotEvaluator_Revive}}
      }
      Evaluate...AvoidThreat {
       class=FortAthenaBTService_AIEvaluator
       properties {
        interval=0.5
        nodeName=Evaluate...AvoidThreat
        aIEvaluatorClass {defaultValue=/Game/Athena/AI/Phoebe/Evaluators/Phoebe_Evaluator_AvoidThreat.Phoebe_Evaluator_AvoidThreat_C}
       }
      }
      Evaluate...StealWall {
       class=FortAthenaBTService_AIEvaluator
       properties {interval=10, nodeName=Evaluate...StealWall, aIEvaluatorClass {defaultValue=/Script/FortniteAIServer.FortAthenaAIBotEvaluator_StealWall}}
      }
      Evaluate...Escape {
       class=FortAthenaBTService_AIEvaluator
       properties {
        interval=0.5
        nodeName=Evaluate...Escape
        aIEvaluatorClass {defaultValue=/Game/Athena/AI/Phoebe/Evaluators/Phoebe_Evaluator_Escape.Phoebe_Evaluator_Escape_C}
       }
      }
      Executing...Pick Up Loot {
       class=FortAthenaBTService_PickUpLoot
       properties {
        bCallTickOnSearchStart=false
        bRestartTimerOnEachActivation=false
        executionStatusName=AIEvaluator_Loot_ExecutionStatus
        interactionContextInfoName=AIEvaluator_Loot_InteractionContextInfo
        interactionExecutionStatusName=AIEvaluator_Loot_InteractionExecutionStatus
        interval=0.5
        lootObjectKeyName=AIEvaluator_Loot_Object
        movementStateKeyName=AIEvaluator_Loot_MovementState
        nodeName=Executing...Pick Up Loot
        randomDeviation=0.1
       }
      }
      Executing...Crouch {class=FortAthenaBTService_Crouch}
      Executing...Slide {
       class=FortAthenaBTService_Slide
       properties {
        bCallTickOnSearchStart=false
        bRestartTimerOnEachActivation=false
        interval=0.5
        nodeName=Executing...Slide
        randomDeviation=0.1
        slideExecutionStatusName=AIEvaluator_Slide_ExecutionStatus
       }
      }
      Executing...Jump {class=FortAthenaBTService_Jump, properties {crouchExecutionStatusName=AIEvaluator_Crouch_ExecutionStatus}}
      Executing... Clamber {
       class=FortAthenaBTService_Clamber
       properties {
        bCallTickOnSearchStart=false
        bRestartTimerOnEachActivation=false
        clamberAbilityStatusName=AIEvaluator_Clamber_AbilityStatus
        clamberAttemptRetryMaxCount=2
        clamberDestinationLocationName=AIEvaluator_Clamber_Destination
        clamberExecutionStatusName=AIEvaluator_Clamber_ExecutionStatus
        clamberOriginLocationName=AIEvaluator_Clamber_Origin
        clamberRetryDelay=0.15
        crouchExecutionStatusName=AIEvaluator_Crouch_ExecutionStatus
        firstJumpClamberMaxStartDelay=2
        firstJumpRetryDelay=0.1
        firstJumpRetryMaxCount=5
        interval=0.5
        jumpExecutionStatusName=AIEvaluator_Jump_ExecutionStatus
        nodeName=Executing... Clamber
        randomDeviation=0.1
       }
      }
      Executing...Sprinting {class=FortAthenaBTService_Sprinting}
      Executing...Build Construction {
       class=FortAthenaBTService_BuildConstruction
       properties {
        bCallTickOnSearchStart=false
        bRestartTimerOnEachActivation=false
        interval=0.5
        nodeName=Executing...Build Construction
        randomDeviation=0.1
        stealWallBuildGridCoordName=AIEvaluator_StealWallBuilding_GridCoord
        stealWallBuildName=AIEvaluator_StealWallBuilding_ExecutionStatus
        stealWallBuildTypeName=AIEvaluator_StealWallBuilding_Type
       }
      }
      Excuting...Escape {
       class=FortAthenaBTService_Escape
       properties {
        bCallTickOnSearchStart=false
        bRestartTimerOnEachActivation=false
        escapeFromStormKeyName=AIEvaluator_Storm_ExecutionStatus
        escapeKeyName=AIEvaluator_Escape_ExecutionStatus
        interval=0.5
        nodeName=Excuting...Escape
        randomDeviation=0.1
       }
      }
      Evaluating...Danger {
       class=FortAthenaBTService_AIEvaluator
       properties {
        interval=1000000000
        nodeName=Evaluating...Danger
        randomDeviation=0
        aIEvaluatorClass {defaultValue=/Game/Athena/AI/Phoebe/Evaluators/Phoebe_Evaluator_DangerDetection.Phoebe_Evaluator_DangerDetection_C}
       }
      }
      Evaluating...Trap On Path {
       class=FortAthenaBTService_AIEvaluator
       properties {
        interval=0.3
        nodeName=Evaluating...Trap On Path
        aIEvaluatorClass {defaultValue=/Script/FortniteAIServer.FortAthenaAIBotEvaluator_TrapOnPathDetected}
       }
      }
      Evaluate...Sand Tunnel {
       class=FortAthenaBTService_AIEvaluator
       properties {
        interval=100000000
        nodeName=Evaluate...Sand Tunnel
        aIEvaluatorClass {defaultValue=/Game/Athena/AI/Phoebe/Evaluators/Phoebe_Evaluator_SandTunnel.Phoebe_Evaluator_SandTunnel_C}
       }
      }
      Evaluating...Ground {
       class=FortAthenaBTService_AIEvaluator
       properties {interval=1, nodeName=Evaluating...Ground, aIEvaluatorClass {defaultValue=/Script/FortniteAIServer.FortAthenaAIBotEvaluator_Ground}}
      }
      Evaluating...Retreat {
       class=FortAthenaBTService_AIEvaluator
       properties {interval=0.5, nodeName=Evaluating...Retreat, aIEvaluatorClass {defaultValue=/Script/FortniteAIServer.FortAthenaAIBotEvaluator_Retreat}}
      }
      "Evaluating... Slot: Traversal" {
       class=FortAthenaBTService_AIEvaluator
       properties {
        interval=100000000
        nodeName=Evaluating... Slot: Traversal
        aIEvaluatorClass {defaultValue=null}
        injectionTag {tagName=Athena.AI.BTInjection.Phoebe.Traversal}
       }
      }
      "Evaluating... Slot: ChromeBlob" {
       class=FortAthenaBTService_AIEvaluator
       properties {
        interval=100000000
        nodeName=Evaluating... Slot: ChromeBlob
        aIEvaluatorClass {defaultValue=null}
        injectionTag {tagName=Athena.AI.BTInjection.Phoebe.Blob}
       }
      }
      "Evaluating... Slot: Grind" {
       class=FortAthenaBTService_AIEvaluator
       properties {
        interval=100000000
        nodeName=Evaluating... Slot: Grind
        aIEvaluatorClass {defaultValue=null}
        injectionTag {tagName=Athena.AI.BTInjection.Phoebe.Grind}
       }
      }
      "Evaluating... Slot: Falling" {
       class=FortAthenaBTService_AIEvaluator
       properties {
        interval=100000000
        nodeName=Evaluating... Slot: Falling
        aIEvaluatorClass {defaultValue=null}
        injectionTag {tagName=Behavior.Injection.Alive.Falling}
       }
      }
     ]
    }
    Fallback - Wait {class=BTTask_Wait, properties {nodeName=Fallback - Wait, waitTime {defaultValue=5}}}
   ]
   decorators []
   properties {
    nodeName=Alive
    children [
     {
      childComposite=null
      childTask=&self:FortBTTask_InjectionPoint_1
      decorators [&self:FortBTDecorator_CanDoInjectedBehavior_0]
     }
     {
      childComposite=null
      childTask=&self:FortAthenaBTTask_Zipline_0
      decorators [&self:BTDecorator_Blackboard_34]
     }
     {
      childComposite=null
      childTask=&self:FortAthenaBTTask_Dive_0
      decorators [&self:BTDecorator_Blackboard_21]
     }
     {
      childComposite=null
      childTask=&self:FortAthenaBTTask_Glide_0
      decorators [&self:BTDecorator_Blackboard_22]
     }
     {
      childComposite=null
      childTask=&self:FortBTTask_InjectionPoint_0
      decorators [&self:FortBTDecorator_CanDoInjectedBehavior_4]
     }
     {
      childComposite=&self:BTComposite_Selector_23
      childTask=null
      decorators [&self:BTDecorator_HasEverSkydivedFromBusAndLanded_C_0]
     }
     {childComposite=null, childTask=&self:BTTask_Wait_3, decorators []}
    ]
    services [
     &self:FortAthenaBTService_AIEvaluator_1
     &self:FortBTService_AIEvaluatorsInjector_3
    ]
   }
   services [
    Evaluating...Free Fall {
     class=FortAthenaBTService_AIEvaluator
     properties {
      bCallTickOnSearchStart=true
      interval=0.5
      nodeName=Evaluating...Free Fall
      aIEvaluatorClass {defaultValue=/Game/Athena/AI/Phoebe/Evaluators/Phoebe_Evaluator_FreeFall.Phoebe_Evaluator_FreeFall_C}
      injectionTag {tagName=Behavior.Injection.Alive.FreeFall}
     }
    }
    Behavior Injector {
     class=FortBTService_AIEvaluatorsInjector
     properties {
      bCallTickOnSearchStart=false
      bRestartTimerOnEachActivation=false
      interval=0.5
      nodeName=Behavior Injector
      randomDeviation=0.1
      injectionTagsKeys [
       {
        actionInterfaceObjectKeyName=Injection_Alive_HighPriority_ActionObject
        injectionKeyName=AIEvaluator_Injection_Alive_HighPriority
        injectionTag {tagName=Behavior.Injection.Alive.HighPriority}
       }
       {
        actionInterfaceObjectKeyName=Injection_General_HighPriority_ActionObject
        injectionKeyName=AIEvaluator_Injection_General_HighPriority
        injectionTag {tagName=Behavior.Injection.General.HighPriority}
       }
      ]
     }
    }
   ]
  }
 ]
 decorators []
 properties {
  nodeName=Gameplay
  children [
   {
    childComposite=&self:BTComposite_Selector_12
    childTask=null
    decorators [
     &self:BTDecorator_Blackboard_15
     &self:BTDecorator_Loop_10
    ]
   }
   {
    childComposite=null
    childTask=&self:FortAthenaBTTask_SteerMovement_0
    decorators [
     &self:FortAthenaBTDecorator_BehaviorControls_23
     &self:BTDecorator_Blackboard_12
    ]
   }
   {
    childComposite=null
    childTask=&self:BTTask_RunBehavior_5
    decorators [&self:BTDecorator_Blackboard_23]
   }
   {childComposite=&self:BTComposite_Selector_10, childTask=null, decorators []}
  ]
  services [
   &self:FortAthenaBTService_UpdateTarget_4
   &self:FortAthenaBTService_AIEvaluator_65
   &self:FortAthenaBTService_AIEvaluator_66
  ]
 }
 services [
  Evaluating...Targets {class=FortAthenaBTService_UpdateTarget, properties {interval=0.1}}
  Evaluating...Escape Storm {
   class=FortAthenaBTService_AIEvaluator
   properties {
    interval=1
    nodeName=Evaluating...Escape Storm
    randomDeviation=0.5
    aIEvaluatorClass {defaultValue=/Game/Athena/AI/Phoebe/Evaluators/Phoebe_Evaluator_Storm.Phoebe_Evaluator_Storm_C}
   }
  }
  Evaluating...Character Launched {
   class=FortAthenaBTService_AIEvaluator
   properties {
    interval=0.5
    nodeName=Evaluating...Character Launched
    aIEvaluatorClass {defaultValue=/Game/Athena/AI/Phoebe/Evaluators/Phoebe_Evaluator_CharacterLaunched.Phoebe_Evaluator_CharacterLaunched_C}
   }
  }
 ]
}
