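# Behavior tree for Phoebe AI bots during match warmup, choosing between emoting, looting and shooting, or idling AFK.
# A root Warmup evaluator and target updater run throughout; evaluator services feed blackboard keys that gate branches.
# The emote branch moves to a destination and emotes, otherwise waits.
# Loot-and-shoot runs MoveToRange or Loot subtrees when threatened, else loops looting, with services handling weapons, pickup, crouch, jump and sprint.
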BehaviorTree /Game/Athena/AI/Phoebe/BehaviorTrees/BT_Phoebe_Warmup
blackboard=&BB_Phoebe
root {
 name=Selector
 class=BTComposite_Selector
 children [
  Sequence {
   class=BTComposite_Sequence
   children [Selector {
     class=BTComposite_Selector
     children [
      Sequence {
       class=BTComposite_Sequence
       children [
        PlayEmote Move To {
         class=FortAthenaBTTask_BotMoveTo
         decorators []
         properties {
          movementResultKeyName=AIEvaluator_PlayEmote_MovementState
          nodeName=PlayEmote Move To
          acceptableRadius {defaultValue=500}
          allowRandomWobble {defaultValue=false}
          bAllowStrafe {defaultValue=true}
          blackboardKey {selectedKeyName=AIEvaluator_PlayEmote_Destination}
          observedBlackboardValueTolerance {defaultValue=4.75}
         }
        }
        PlayEmote {
         class=FortAthenaBTTask_PlayEmote
         children []
         decorators []
         properties {bIgnoreRestartSelf=false, nodeName=PlayEmote, playEmoteExecutionStatusKeyName=AIEvaluator_PlayEmote_ExecutionStatus, bespokeEmotes []}
         services []
        }
       ]
       decorators [Play Emote? {
         class=BTDecorator_Blackboard
         properties {
          arithmeticOperation=GreaterOrEqual
          intValue=3
          nodeName=Play Emote?
          stringValue=EExecutionStatus::ExecutionPending
          blackboardKey {selectedKeyName=AIEvaluator_PlayEmote_ExecutionStatus}
         }
        }]
       properties {children [
         {childTask=&self:FortAthenaBTTask_BotMoveTo_2}
         {childTask=&self:FortAthenaBTTask_PlayEmote_0}
        ]}
      }
      Wait {class=BTTask_Wait, properties {randomDeviation {defaultValue=3}, waitTime {defaultValue=5}}}
     ]
     decorators []
     properties {
      children [
       {
        childComposite=&self:BTComposite_Sequence_7
        childTask=null
        decorators [&self:BTDecorator_Blackboard_61]
       }
       {childComposite=null, childTask=&self:BTTask_Wait_14, decorators []}
      ]
      services [&self:FortAthenaBTService_AIEvaluator_0]
     }
     services [Evaluating...PlayEmote {
       class=FortAthenaBTService_AIEvaluator
       properties {
        bCallTickOnSearchStart=true
        interval=0.5
        nodeName=Evaluating...PlayEmote
        aIEvaluatorClass {defaultValue=/Script/FortniteAIServer.FortAthenaAIBotEvaluator_PlayEmote}
       }
      }]
    }]
   decorators [Warmup PlayEmote? {
     class=BTDecorator_Blackboard
     properties {
      arithmeticOperation=GreaterOrEqual
      intValue=3
      nodeName=Warmup PlayEmote?
      stringValue=EExecutionStatus::ExecutionPending
      blackboardKey {selectedKeyName=AIEvaluator_WarmupPlayEmote_ExecutionStatus}
     }
    }]
   properties {children [{childComposite=&self:BTComposite_Selector_2, childTask=null}]}
  }
  On Ground {
   class=BTComposite_Selector
   children [
    THREATEN {
     class=BTComposite_Selector
     children [
      Move To Range {
       class=BTTask_RunBehavior
       decorators [
        Can do Range {
         class=BTDecorator_Blackboard
         properties {
          arithmeticOperation=GreaterOrEqual
          intValue=3
          nodeName=Can do Range
          stringValue=EExecutionStatus::ExecutionPending
          blackboardKey {selectedKeyName=AIEvaluator_RangeAttack_ExecutionStatus}
         }
        }
        Loop {class=BTDecorator_Loop}
       ]
       properties {behaviorAsset=&BT_Phoebe_MoveToRange, nodeName=Move To Range}
      }
      Run Loot Behavior {
       class=BTTask_RunBehavior
       decorators [Loot {
         class=BTDecorator_Blackboard
         properties {
          arithmeticOperation=GreaterOrEqual
          intValue=2
          nodeName=Loot
          stringValue=EExecutionStatus::ExecutionSuccess
          blackboardKey {selectedKeyName=AIEvaluator_Loot_ExecutionStatus}
         }
        }]
       properties {behaviorAsset=&/BT_Phoebe_Loot, nodeName=Run Loot Behavior}
      }
     ]
     decorators [IsThreatened {
       class=BTDecorator_Blackboard
       properties {
        arithmeticOperation=Equal
        intValue=3
        nodeName=IsThreatened
        stringValue=EAlertLevel::Threatened
        blackboardKey {selectedKeyName=AIEvaluator_Global_AlertLevel}
       }
      }]
     properties {
      nodeName=THREATEN
      children [
       {
        childComposite=null
        childTask=&self:BTTask_RunBehavior_0
        decorators [
         &self:BTDecorator_Blackboard_1
         &self:BTDecorator_Loop_1
        ]
       }
       {
        childComposite=null
        childTask=&self:BTTask_RunBehavior_25
        decorators [&self:BTDecorator_Blackboard_51]
       }
      ]
      services [
       &self:FortAthenaBTService_AIEvaluator_91
       &self:FortAthenaBTService_AIEvaluator_92
       &self:FortAthenaBTService_ManageWeapon_8
      ]
     }
     services [
      Evaluating...Range Attack {
       class=FortAthenaBTService_AIEvaluator
       properties {
        bCallTickOnSearchStart=true
        interval=0.2
        nodeName=Evaluating...Range Attack
        randomDeviation=0.05
        aIEvaluatorClass {defaultValue=&FortAthenaAIBotEvaluator_RangeAttack}
       }
      }
      Evaluating...Melee Attack {
       class=FortAthenaBTService_AIEvaluator
       properties {
        bCallTickOnSearchStart=true
        interval=1
        nodeName=Evaluating...Melee Attack
        aIEvaluatorClass {defaultValue=&FortAthenaAIBotEvaluator_MeleeAttack}
       }
      }
      Executing...Weapons Actions {class=FortAthenaBTService_ManageWeapon, properties {interval=1, randomDeviation=0.3}}
     ]
    }
    UNAWARE {
     class=BTComposite_Selector
     children [Run Loot Behavior {
       class=BTTask_RunBehavior
       decorators [Loot {
         class=BTDecorator_Blackboard
         properties {
          arithmeticOperation=GreaterOrEqual
          intValue=2
          nodeName=Loot
          stringValue=EExecutionStatus::ExecutionSuccess
          blackboardKey {selectedKeyName=AIEvaluator_Loot_ExecutionStatus}
         }
        }]
       properties {behaviorAsset=&BT_Phoebe_Loot, nodeName=Run Loot Behavior}
      }]
     decorators [Loop {class=BTDecorator_Loop}]
     properties {
      nodeName=UNAWARE
      children [{
        childComposite=null
        childTask=&self:BTTask_RunBehavior_26
        decorators [&self:BTDecorator_Blackboard_50]
       }]
      services [
       &self:FortAthenaBTService_AIEvaluator_93
       &self:FortAthenaBTService_AIEvaluator_94
       &self:FortAthenaBTService_ManageWeapon_9
       &self:FortAthenaBTService_AIEvaluator_95
      ]
     }
     services [
      Evaluating...Range Attack {
       class=FortAthenaBTService_AIEvaluator
       properties {
        interval=0.3
        nodeName=Evaluating...Range Attack
        aIEvaluatorClass {defaultValue=&FortAthenaAIBotEvaluator_RangeAttack}
       }
      }
      Evaluating...Melee Attack {
       class=FortAthenaBTService_AIEvaluator
       properties {
        interval=0.3
        nodeName=Evaluating...Melee Attack
        aIEvaluatorClass {defaultValue=&FortAthenaAIBotEvaluator_MeleeAttack}
       }
      }
      Executing...Weapons Actions {class=FortAthenaBTService_ManageWeapon, properties {interval=0.2, randomDeviation=0.1}}
      Focus Target Ignore Threat {
       class=FortAthenaBTService_AIEvaluator
       properties {
        interval=0.5
        nodeName=Focus Target Ignore Threat
        aIEvaluatorClass {defaultValue=/Game/Athena/AI/Phoebe/Evaluators/Phoebe_HandleFocusing_IgnoreThreat.Phoebe_HandleFocusing_IgnoreThreat_C}
       }
      }
     ]
    }
    Wait {class=BTTask_Wait, properties {waitTime {defaultValue=5}}}
   ]
   decorators [Warmup LootAndShoot? {
     class=BTDecorator_Blackboard
     properties {
      arithmeticOperation=GreaterOrEqual
      intValue=3
      nodeName=Warmup LootAndShoot?
      stringValue=EExecutionStatus::ExecutionPending
      blackboardKey {selectedKeyName=AIEvaluator_WarmupLootAndShoot_ExecutionStatus}
     }
    }]
   properties {
    nodeName=On Ground
    children [
     {
      childComposite=&self:BTComposite_Selector_36
      childTask=null
      decorators [&self:BTDecorator_Blackboard_60]
     }
     {
      childComposite=&self:BTComposite_Selector_37
      childTask=null
      decorators [&self:BTDecorator_Loop_11]
     }
     {childComposite=null, childTask=&self:BTTask_Wait_15, decorators []}
    ]
    services [
     &self:FortAthenaBTService_AIEvaluator_17
     &self:FortAthenaBTService_AIEvaluator_86
     &self:FortAthenaBTService_AIEvaluator_87
     &self:FortAthenaBTService_PickUpLoot_2
     &self:FortAthenaBTService_Crouch_2
     &self:FortAthenaBTService_Jump_2
     &self:FortAthenaBTService_Sprinting_2
    ]
   }
   services [
    Evaluating...Select Weapon {
     class=FortAthenaBTService_AIEvaluator
     properties {
      interval=0.2
      nodeName=Evaluating...Select Weapon
      randomDeviation=0.05
      aIEvaluatorClass {defaultValue=&FortAthenaAIBotEvaluator_WeaponSelection}
     }
    }
    Evaluate...Loot {
     class=FortAthenaBTService_AIEvaluator
     properties {
      bCallTickOnSearchStart=true
      interval=0.5
      nodeName=Evaluate...Loot
      aIEvaluatorClass {defaultValue=/Game/Athena/AI/Phoebe/Evaluators/Phoebe_Evaluator_WarmupLoot.Phoebe_Evaluator_WarmupLoot_C}
     }
    }
    Evaluate...Sprinting {
     class=FortAthenaBTService_AIEvaluator
     properties {interval=0.2, nodeName=Evaluate...Sprinting, aIEvaluatorClass {defaultValue=/Script/FortniteAIServer.FortAthenaAIBotEvaluator_Sprinting}}
    }
    Executing...Pick Up Loot {
     class=FortAthenaBTService_PickUpLoot
     properties {
      bCallTickOnSearchStart=false
      bRestartTimerOnEachActivation=false
      executionStatusName=AIEvaluator_Loot_ExecutionStatus
      interactionContextInfoName=AIEvaluator_Loot_InteractionContextInfo
      interactionExecutionStatusName=AIEvaluator_Loot_InteractionExecutionStatus
      interval=0.5
      lootObjectKeyName=AIEvaluator_Loot_Object
      movementStateKeyName=AIEvaluator_Loot_MovementState
      nodeName=Executing...Pick Up Loot
      randomDeviation=0.1
     }
    }
    Executing...Crouch {class=FortAthenaBTService_Crouch}
    Executing...Jump {class=FortAthenaBTService_Jump, properties {crouchExecutionStatusName=AIEvaluator_Crouch_ExecutionStatus}}
    Executing...Sprinting {class=FortAthenaBTService_Sprinting}
   ]
  }
  Warmup AFK {class=BTTask_Wait, properties {nodeName=Warmup AFK, waitTime {defaultValue=99999}}}
 ]
 decorators []
 properties {
  children [
   {
    childComposite=&self:BTComposite_Sequence_1
    childTask=null
    decorators [&self:BTDecorator_Blackboard_3]
   }
   {
    childComposite=&self:BTComposite_Selector_35
    childTask=null
    decorators [&self:BTDecorator_Blackboard_62]
   }
   {childComposite=null, childTask=&self:BTTask_Wait_0, decorators []}
  ]
  services [
   &self:FortAthenaBTService_AIEvaluator_1
   &self:FortAthenaBTService_UpdateTarget_0
  ]
 }
 services [
  Evaluating... Warmup {
   class=FortAthenaBTService_AIEvaluator
   properties {interval=0.5, nodeName=Evaluating... Warmup, aIEvaluatorClass {defaultValue=/Script/FortniteAIServer.FortAthenaAIBotEvaluator_Warmup}}
  }
  Evaluating...Targets {class=FortAthenaBTService_UpdateTarget, properties {interval=2, randomDeviation=1}}
 ]
}
