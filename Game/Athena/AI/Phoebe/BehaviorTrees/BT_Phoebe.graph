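# Root behavior tree for the Phoebe AI on BB_Phoebe, selecting behavior by game phase: warmup, on bus, end game or gameplay, with a short wait as fallback.
# Each phase branch runs an injector service that feeds an injection point, letting injected behavior take priority over the default through a both-way abort decorator.
# Warmup and gameplay default to BT_Phoebe_Warmup and BT_Phoebe_Gameplay subtrees, while bus and end game just wait.
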BehaviorTree /Game/Athena/AI/Phoebe/BehaviorTrees/BT_Phoebe
blackboard=&BB_Phoebe
root {
 name=Game Phase
 class=BTComposite_Selector
 children [
  Selector {
   class=BTComposite_Selector
   children [
    Injection Point General.Warmup {
     class=FortBTTask_InjectionPoint
     children []
     decorators [Can Do Injected Behavior {
       class=FortBTDecorator_CanDoInjectedBehavior
       properties {
        arithmeticOperation=GreaterOrEqual
        bInverseCondition=false
        basicOperation=Set
        floatValue=0
        flowAbortMode=Both
        intValue=3
        nodeName=Can Do Injected Behavior
        notifyObserver=ResultChange
        stringValue=ExecutionPending
        textOperation=Equal
        blackboardKey {bNoneIsAllowedValue=false, selectedKeyName=AIEvaluator_Injection_General_Warmup}
       }
      }]
     properties {
      bIgnoreRestartSelf=false
      nodeName=Injection Point General.Warmup
      injectedAction {baseClass=&Object, defaultValue=null, key=Injection_General_Warmup_ActionObject}
      injectionTag {tagName=Behavior.Injection.General.Warmup}
     }
     services []
    }
    Warmup {class=BTTask_RunBehavior, decorators [], properties {behaviorAsset=&/BT_Phoebe_Warmup, nodeName=Warmup}}
   ]
   decorators [In Warmup? {
     class=BTDecorator_Blackboard
     properties {
      arithmeticOperation=Equal
      intValue=1
      nodeName=In Warmup?
      stringValue=Warmup
      blackboardKey {selectedKeyName=AIEvaluator_Global_GamePhaseBehavior}
     }
    }]
   properties {
    children [
     {
      childComposite=null
      childTask=&self:FortBTTask_InjectionPoint_0
      decorators [&self:FortBTDecorator_CanDoInjectedBehavior_2]
     }
     {childComposite=null, childTask=&self:BTTask_RunBehavior_5, decorators []}
    ]
    services [&self:FortBTService_AIEvaluatorsInjector_2]
   }
   services [Behavior Injector {
     class=FortBTService_AIEvaluatorsInjector
     properties {
      bCallTickOnSearchStart=false
      bRestartTimerOnEachActivation=false
      interval=0.5
      nodeName=Behavior Injector
      randomDeviation=0.1
      injectionTagsKeys [{
        actionInterfaceObjectKeyName=Injection_General_Warmup_ActionObject
        injectionKeyName=AIEvaluator_Injection_General_Warmup
        injectionTag {tagName=Behavior.Injection.General.Warmup}
       }]
     }
    }]
  }
  Selector {
   class=BTComposite_Selector
   children [
    Injection Point General.OnBus {
     class=FortBTTask_InjectionPoint
     children []
     decorators [Can Do Injected Behavior {
       class=FortBTDecorator_CanDoInjectedBehavior
       properties {
        arithmeticOperation=GreaterOrEqual
        bInverseCondition=false
        basicOperation=Set
        floatValue=0
        flowAbortMode=Both
        intValue=3
        nodeName=Can Do Injected Behavior
        notifyObserver=ResultChange
        stringValue=ExecutionPending
        textOperation=Equal
        blackboardKey {bNoneIsAllowedValue=false, selectedKeyName=AIEvaluator_Injection_General_OnBus}
       }
      }]
     properties {
      bIgnoreRestartSelf=false
      nodeName=Injection Point General.OnBus
      injectedAction {baseClass=&Object, defaultValue=null, key=Injection_General_OnBus_ActionObject}
      injectionTag {tagName=Behavior.Injection.General.OnBus}
     }
     services []
    }
    Wait {
     class=BTTask_Wait
     decorators [On Bus Locked? {
       class=BTDecorator_Blackboard
       properties {
        arithmeticOperation=Equal
        intValue=4
        nodeName=On Bus Locked?
        stringValue=EAthenaGamePhaseStep::BusLocked
        blackboardKey {selectedKeyName=AIEvaluator_Global_GamePhaseStep}
       }
      }]
     properties {waitTime {defaultValue=999}}
    }
    Wait {class=BTTask_Wait, properties {waitTime {defaultValue=999}}}
    Wait {class=BTTask_Wait, properties {waitTime {defaultValue=1}}}
   ]
   decorators [Is On Bus? {
     class=BTDecorator_Blackboard
     properties {arithmeticOperation=Equal, intValue=0, nodeName=Is On Bus?, stringValue="", blackboardKey {selectedKeyName=AIEvaluator_Global_IsInBus}}
    }]
   properties {
    children [
     {
      childComposite=null
      childTask=&self:FortBTTask_InjectionPoint_1
      decorators [&self:FortBTDecorator_CanDoInjectedBehavior_1]
     }
     {
      childComposite=null
      childTask=&self:BTTask_Wait_6
      decorators [&self:BTDecorator_Blackboard_16]
     }
     {childComposite=null, childTask=&self:BTTask_Wait_3, decorators []}
     {childComposite=null, childTask=&self:BTTask_Wait_1, decorators []}
    ]
    services [&self:FortBTService_AIEvaluatorsInjector_1]
   }
   services [Behavior Injector {
     class=FortBTService_AIEvaluatorsInjector
     properties {
      bCallTickOnSearchStart=false
      bRestartTimerOnEachActivation=false
      interval=0.5
      nodeName=Behavior Injector
      randomDeviation=0.1
      injectionTagsKeys [{
        actionInterfaceObjectKeyName=Injection_General_OnBus_ActionObject
        injectionKeyName=AIEvaluator_Injection_General_OnBus
        injectionTag {tagName=Behavior.Injection.General.OnBus}
       }]
     }
    }]
  }
  Selector {
   class=BTComposite_Selector
   children [
    Injection Point General.EndGame {
     class=FortBTTask_InjectionPoint
     children []
     decorators [Can Do Injected Behavior {
       class=FortBTDecorator_CanDoInjectedBehavior
       properties {
        arithmeticOperation=GreaterOrEqual
        bInverseCondition=false
        basicOperation=Set
        floatValue=0
        flowAbortMode=Both
        intValue=3
        nodeName=Can Do Injected Behavior
        notifyObserver=ResultChange
        stringValue=ExecutionPending
        textOperation=Equal
        blackboardKey {bNoneIsAllowedValue=false, selectedKeyName=AIEvaluator_Injection_General_EndGame}
       }
      }]
     properties {
      bIgnoreRestartSelf=false
      nodeName=Injection Point General.EndGame
      injectedAction {baseClass=&Object, defaultValue=null, key=Injection_General_EndGame_ActionObject}
      injectionTag {tagName=Behavior.Injection.General.EndGame}
     }
     services []
    }
    Wait {class=BTTask_Wait, properties {waitTime {defaultValue=60}}}
   ]
   decorators [Use End Game Behavior? {
     class=BTDecorator_Blackboard
     properties {
      arithmeticOperation=Equal
      intValue=3
      nodeName=Use End Game Behavior?
      stringValue=EndGame
      blackboardKey {selectedKeyName=AIEvaluator_Global_GamePhaseBehavior}
     }
    }]
   properties {
    children [
     {
      childComposite=null
      childTask=&self:FortBTTask_InjectionPoint_3
      decorators [&self:FortBTDecorator_CanDoInjectedBehavior_3]
     }
     {childComposite=null, childTask=&self:BTTask_Wait_4, decorators []}
    ]
    services [&self:FortBTService_AIEvaluatorsInjector_4]
   }
   services [Behavior Injector {
     class=FortBTService_AIEvaluatorsInjector
     properties {
      bCallTickOnSearchStart=false
      bRestartTimerOnEachActivation=false
      interval=0.5
      nodeName=Behavior Injector
      randomDeviation=0.1
      injectionTagsKeys [{
        actionInterfaceObjectKeyName=Injection_General_EndGame_ActionObject
        injectionKeyName=AIEvaluator_Injection_General_EndGame
        injectionTag {tagName=Behavior.Injection.General.EndGame}
       }]
     }
    }]
  }
  Selector {
   class=BTComposite_Selector
   children [
    Injection Point General.Gameplay {
     class=FortBTTask_InjectionPoint
     children []
     decorators [Can Do Injected Behavior {
       class=FortBTDecorator_CanDoInjectedBehavior
       properties {
        arithmeticOperation=GreaterOrEqual
        bInverseCondition=false
        basicOperation=Set
        floatValue=0
        flowAbortMode=Both
        intValue=3
        nodeName=Can Do Injected Behavior
        notifyObserver=ResultChange
        stringValue=ExecutionPending
        textOperation=Equal
        blackboardKey {bNoneIsAllowedValue=false, selectedKeyName=AIEvaluator_Injection_General_Gameplay}
       }
      }]
     properties {
      bIgnoreRestartSelf=false
      nodeName=Injection Point General.Gameplay
      injectedAction {baseClass=&Object, defaultValue=null, key=Injection_General_Gameplay_ActionObject}
      injectionTag {tagName=Behavior.Injection.General.Gameplay}
     }
     services []
    }
    Run Behavior {
     class=BTTask_RunBehavior
     decorators []
     properties {behaviorAsset=&/BT_Phoebe_Gameplay, nodeName=Run Behavior}
    }
   ]
   decorators [In Gameplay? {
     class=BTDecorator_Blackboard
     properties {
      arithmeticOperation=Equal
      intValue=2
      nodeName=In Gameplay?
      stringValue=Gameplay
      blackboardKey {selectedKeyName=AIEvaluator_Global_GamePhaseBehavior}
     }
    }]
   properties {
    children [
     {
      childComposite=null
      childTask=&self:FortBTTask_InjectionPoint_2
      decorators [&self:FortBTDecorator_CanDoInjectedBehavior_0]
     }
     {childComposite=null, childTask=&self:BTTask_RunBehavior_0, decorators []}
    ]
    services [&self:FortBTService_AIEvaluatorsInjector_0]
   }
   services [Behavior Injector {
     class=FortBTService_AIEvaluatorsInjector
     properties {
      bCallTickOnSearchStart=false
      bRestartTimerOnEachActivation=false
      interval=0.5
      nodeName=Behavior Injector
      randomDeviation=0.1
      injectionTagsKeys [{
        actionInterfaceObjectKeyName=Injection_General_Gameplay_ActionObject
        injectionKeyName=AIEvaluator_Injection_General_Gameplay
        injectionTag {tagName=Behavior.Injection.General.Gameplay}
       }]
     }
    }]
  }
  Wait {class=BTTask_Wait, properties {waitTime {defaultValue=1}}}
 ]
 decorators []
 properties {
  nodeName=Game Phase
  children [
   {
    childComposite=&self:BTComposite_Selector_0
    childTask=null
    decorators [&self:BTDecorator_Blackboard_2]
   }
   {
    childComposite=&self:BTComposite_Selector_7
    childTask=null
    decorators [&self:BTDecorator_Blackboard_17]
   }
   {
    childComposite=&self:BTComposite_Selector_3
    childTask=null
    decorators [&self:BTDecorator_Blackboard_1]
   }
   {
    childComposite=&self:BTComposite_Selector_2
    childTask=null
    decorators [&self:BTDecorator_Blackboard_3]
   }
   {childComposite=null, childTask=&self:BTTask_Wait_0, decorators []}
  ]
  services []
 }
 services []
}
