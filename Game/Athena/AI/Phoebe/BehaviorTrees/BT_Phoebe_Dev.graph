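# Dev behavior tree for the Phoebe AI bot in Fortnite Athena, driving gameplay from the Aircraft phase onward.
# A priority selector handles blocked movement, launches, DBNO subtree, zipline, dive, glide, then grounded trap, threat and danger responses.
# Evaluator services write blackboard keys that gate tasks and feed executor services for looting, movement, building and weapons.
# Alerted combat runs through services over looping Waits, and tagged empty evaluator slots accept injected Traversal, ChromeBlob and Grind logic.
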BehaviorTree /Game/Athena/AI/Phoebe/BehaviorTrees/BT_Phoebe_Dev
blackboard=&BB_Phoebe
root {
 name=Gameplay
 class=BTComposite_Selector
 children [
  Wait {
   class=BTTask_Wait
   decorators [
    Is Movement Blocked? {
     class=BTDecorator_Blackboard
     properties {
      arithmeticOperation=Equal
      intValue=0
      nodeName=Is Movement Blocked?
      stringValue=""
      blackboardKey {selectedKeyName=AIEvaluator_Global_IsMovementBlocked}
     }
    }
    Loop {class=BTDecorator_Loop}
   ]
   properties {waitTime {defaultValue=1}}
  }
  Steer Movement {
   class=FortAthenaBTTask_SteerMovement
   children []
   decorators [Is Character being launched? {
     class=BTDecorator_Blackboard
     properties {
      arithmeticOperation=GreaterOrEqual
      intValue=3
      nodeName=Is Character being launched?
      stringValue=EExecutionStatus::ExecutionPending
      blackboardKey {selectedKeyName=AIEvaluator_CharacterLaunched_ExecutionStatus}
     }
    }]
   properties {
    bIgnoreRestartSelf=false
    bSetControlRotation=false
    nodeName=Steer Movement
    steerDirectionKeySelector {bNoneIsAllowedValue=false, selectedKeyName=AIEvaluator_CharacterLaunched_SteerDirection}
   }
   services []
  }
  Run Behavior {
   class=BTTask_RunBehavior
   decorators [Is DBNO {
     class=BTDecorator_Blackboard
     properties {arithmeticOperation=GreaterOrEqual, intValue=3, nodeName=Is DBNO, stringValue="", blackboardKey {selectedKeyName=AIEvaluator_Global_DBNO}}
    }]
   properties {behaviorAsset=&/BT_Phoebe_DBNO, nodeName=Run Behavior}
  }
  Alive {
   class=BTComposite_Selector
   children [
    Ziplining {
     class=FortAthenaBTTask_Zipline
     children []
     decorators [Is on Zipline {
       class=BTDecorator_Blackboard
       properties {
        arithmeticOperation=GreaterOrEqual
        intValue=4
        nodeName=Is on Zipline
        stringValue=EExecutionStatus::ExecutionAllowed
        blackboardKey {selectedKeyName=AIEvaluator_Zipline_UsageExecutionStatus}
       }
      }]
     properties {
      bIgnoreRestartSelf=false
      nodeName=Ziplining
      usageExecutionStatusName=AIEvaluator_Zipline_UsageExecutionStatus
      ziplineTargetName=AIEvaluator_Zipline_Target
     }
     services []
    }
    Diving {
     class=FortAthenaBTTask_Dive
     children []
     decorators [Can Dive {
       class=BTDecorator_Blackboard
       properties {
        arithmeticOperation=Equal
        intValue=4
        nodeName=Can Dive
        stringValue=EExecutionStatus::ExecutionAllowed
        blackboardKey {selectedKeyName=AIEvaluator_Dive_ExecutionStatus}
       }
      }]
     properties {
      bIgnoreRestartSelf=false
      diveDestinationKeyName=AIEvaluator_Dive_Destination
      executionStatusKeyName=AIEvaluator_Dive_ExecutionStatus
      nodeName=Diving
     }
     services []
    }
    Gliding {
     class=FortAthenaBTTask_Glide
     children []
     decorators [Can Glide {
       class=BTDecorator_Blackboard
       properties {
        arithmeticOperation=Equal
        intValue=4
        nodeName=Can Glide
        stringValue=EExecutionStatus::ExecutionAllowed
        blackboardKey {selectedKeyName=AIEvaluator_Glide_ExecutionStatus}
       }
      }]
     properties {
      bIgnoreRestartSelf=false
      executionStatusKeyName=AIEvaluator_Glide_ExecutionStatus
      glideDestinationKeyName=AIEvaluator_Glide_Destination
      nodeName=Gliding
     }
     services []
    }
    On Ground {
     class=BTComposite_Selector
     children [
      Bot Shoots Trap {
       class=FortAthenaBTTask_ShootTrap
       children []
       decorators [Trap on Path? {
         class=BTDecorator_Blackboard
         properties {
          arithmeticOperation=GreaterOrEqual
          intValue=3
          nodeName=Trap on Path?
          stringValue=EExecutionStatus::ExecutionPending
          blackboardKey {selectedKeyName=AIEvaluator_TrapOnPath_ExecutionStatus}
         }
        }]
       properties {
        bIgnoreRestartSelf=false
        nodeName=Bot Shoots Trap
        trapOnPathKeyName=AIEvaluator_TrapOnPath_ExecutionStatus
        randomDeviation {defaultValue=0, key=None}
        targetActorKey {bNoneIsAllowedValue=false, selectedKeyName=AIEvaluator_TrapOnPath_Actor}
        waitTime {defaultValue=5, key=None}
       }
       services []
      }
      AvoidThreat MoveTo {
       class=FortAthenaBTTask_BotMoveTo
       decorators [Can AvoidThreat? {
         class=BTDecorator_Blackboard
         properties {
          arithmeticOperation=GreaterOrEqual
          intValue=4
          nodeName=Can AvoidThreat?
          stringValue=EExecutionStatus::ExecutionAllowed
          blackboardKey {selectedKeyName=AIEvaluator_AvoidThreat_ExecutionStatus}
         }
        }]
       properties {
        movementResultKeyName=AIEvaluator_AvoidThreat_MovementState
        nodeName=AvoidThreat MoveTo
        acceptableRadius {defaultValue=100}
        allowRandomWobble {defaultValue=false}
        bAllowStrafe {defaultValue=true}
        blackboardKey {selectedKeyName=AIEvaluator_AvoidThreat_Destination}
        observedBlackboardValueTolerance {defaultValue=4.75}
       }
      }
      Move away from danger {
       class=FortAthenaBTTask_BotMoveTo
       decorators [Danger detected? {
         class=BTDecorator_Blackboard
         properties {
          arithmeticOperation=GreaterOrEqual
          intValue=3
          nodeName=Danger detected?
          stringValue=EExecutionStatus::ExecutionPending
          blackboardKey {selectedKeyName=AIEvaluator_DangerZoneDetected_ExecutionStatus}
         }
        }]
       properties {
        executionStatusKeyName=AIEvaluator_DangerZoneDetected_ExecutionStatus
        movementResultKeyName=None
        nodeName=Move away from danger
        acceptableRadius {defaultValue=5}
        allowRandomWobble {defaultValue=false}
        bAllowStrafe {defaultValue=true}
        bTrackMovingGoal {defaultValue=false}
        blackboardKey {selectedKeyName=AIEvaluator_DangerZoneDetected_SafeLocation}
        observedBlackboardValueTolerance {defaultValue=4.75}
       }
      }
      AWARE {
       class=BTComposite_Selector
       children [
        Sequence {
         class=BTComposite_Sequence
         children [Wait {class=BTTask_Wait, decorators [Loop {class=BTDecorator_Loop, properties {bInfiniteLoop=false}}], properties {waitTime {defaultValue=2}}}]
         decorators [
          Is Threatened or Lost Target {
           class=BTDecorator_Blackboard
           properties {
            arithmeticOperation=GreaterOrEqual
            intValue=2
            nodeName=Is Threatened or Lost Target
            stringValue=EAlertLevel::LKP
            blackboardKey {selectedKeyName=AIEvaluator_Global_AlertLevel}
           }
          }
          Blackboard Based Condition {
           class=BTDecorator_Blackboard
           properties {
            arithmeticOperation=GreaterOrEqual
            intValue=3
            nodeName=Blackboard Based Condition
            stringValue=EExecutionStatus::ExecutionPending
            blackboardKey {selectedKeyName=AIEvaluator_RangeAttack_ExecutionStatus}
           }
          }
          Blackboard Based Condition {
           class=BTDecorator_Blackboard
           properties {
            arithmeticOperation=GreaterOrEqual
            intValue=3
            nodeName=Blackboard Based Condition
            stringValue=EExecutionStatus::ExecutionPending
            blackboardKey {selectedKeyName=AIEvaluator_ThrowableAttack_ExecutionStatus}
           }
          }
         ]
         properties {
          children [{
            childTask=&self:BTTask_Wait_2
            decorators [&self:BTDecorator_Loop_0]
           }]
          services [
           &self:FortAthenaBTService_AIEvaluator_52
           &self:FortAthenaBTService_AIEvaluator_55
           &self:FortAthenaBTService_AIEvaluator_57
          ]
         }
         services [
          "Evaluating...Evasive Maneuvers (Range Fight)" {
           class=FortAthenaBTService_AIEvaluator
           properties {
            interval=0.5
            nodeName="Evaluating...Evasive Maneuvers (Range Fight)"
            aIEvaluatorClass {defaultValue=/Game/Athena/AI/Phoebe/Evaluators/Phoebe_EvasiveManeuvers_RangeFight.Phoebe_EvasiveManeuvers_RangeFight_C}
           }
          }
          Focus on target {
           class=FortAthenaBTService_AIEvaluator
           properties {interval=0.5, nodeName=Focus on target, aIEvaluatorClass {defaultValue=&FortAthenaAIBotEvaluator_HandleFocusing}}
          }
          Evaluating...AimDownSight {
           class=FortAthenaBTService_AIEvaluator
           properties {
            interval=0.5
            nodeName=Evaluating...AimDownSight
            aIEvaluatorClass {defaultValue=/Script/FortniteAIServer.FortAthenaAIBotEvaluator_AimDownSight}
           }
          }
         ]
        }
        Wait {class=BTTask_Wait, decorators [Loop {class=BTDecorator_Loop, properties {bInfiniteLoop=false}}], properties {waitTime {defaultValue=2}}}
       ]
       decorators ["IsThreatened, LostTarget or Suspicious" {
         class=BTDecorator_Blackboard
         properties {
          arithmeticOperation=GreaterOrEqual
          intValue=1
          nodeName="IsThreatened, LostTarget or Suspicious"
          stringValue=EAlertLevel::Alerted
          blackboardKey {selectedKeyName=AIEvaluator_Global_AlertLevel}
         }
        }]
       properties {
        nodeName=AWARE
        children [
         {
          childComposite=&self:BTComposite_Sequence_1
          childTask=null
          decoratorOps [{number=2, operation=And}, {number=0}, {number=2, operation=Or}, {number=1}, {number=2}]
          decorators [
           &self:BTDecorator_Blackboard_27
           &self:BTDecorator_Blackboard_34
           &self:BTDecorator_Blackboard_29
          ]
         }
         {
          childComposite=null
          childTask=&self:BTTask_Wait_4
          decorators [&self:BTDecorator_Loop_1]
         }
        ]
        services [
         &self:FortAthenaBTService_AIEvaluator_17
         &self:FortAthenaBTService_AIEvaluator_7
         &self:FortAthenaBTService_AIEvaluator_16
         &self:FortAthenaBTService_AIEvaluator_8
         &self:FortAthenaBTService_ManageWeapon_2
         &self:FortAthenaBTService_AIEvaluator_25
         &self:FortAthenaBTService_AIEvaluator_26
         &self:FortAthenaBTService_AIEvaluator_32
         &self:FortAthenaBTService_AIEvaluator_37
         &self:FortAthenaBTService_AIEvaluator_61
        ]
       }
       services [
        Evaluating...Select Weapon {
         class=FortAthenaBTService_AIEvaluator
         properties {
          interval=0.2
          nodeName=Evaluating...Select Weapon
          randomDeviation=0.05
          aIEvaluatorClass {defaultValue=&FortAthenaAIBotEvaluator_WeaponSelection}
         }
        }
        Evaluating...Range Attack {
         class=FortAthenaBTService_AIEvaluator
         properties {
          bCallTickOnSearchStart=true
          interval=0.2
          nodeName=Evaluating...Range Attack
          randomDeviation=0.05
          aIEvaluatorClass {defaultValue=/Game/Athena/AI/Phoebe/Evaluators/Phoebe_Evaluator_RangeAttack.Phoebe_Evaluator_RangeAttack_C}
         }
        }
        Evaluating...Throwable Attack {
         class=FortAthenaBTService_AIEvaluator
         properties {
          interval=0.2
          nodeName=Evaluating...Throwable Attack
          randomDeviation=0.05
          aIEvaluatorClass {defaultValue=/Script/FortniteAIServer.FortAthenaAIBotEvaluator_ThrowableAttack}
         }
        }
        Evaluating...Melee Attack {
         class=FortAthenaBTService_AIEvaluator
         properties {
          bCallTickOnSearchStart=true
          interval=1
          nodeName=Evaluating...Melee Attack
          aIEvaluatorClass {defaultValue=&FortAthenaAIBotEvaluator_MeleeAttack}
         }
        }
        Executing...Weapons Actions {class=FortAthenaBTService_ManageWeapon, properties {interval=1, randomDeviation=0.3}}
        Evaluating...Investigate Treat {
         class=FortAthenaBTService_AIEvaluator
         properties {
          bCallTickOnSearchStart=true
          interval=0.5
          nodeName=Evaluating...Investigate Treat
          aIEvaluatorClass {defaultValue=/Game/Athena/AI/Phoebe/Evaluators/Phoebe_Evaluator_Investigate.Phoebe_Evaluator_Investigate_C}
         }
        }
        Evaluating... Ambush {
         class=FortAthenaBTService_AIEvaluator
         properties {interval=0.5, nodeName=Evaluating... Ambush, aIEvaluatorClass {defaultValue=/Script/FortniteAIServer.FortAthenaAIBotEvaluator_Ambush}}
        }
        Evaluating... Bunker {
         class=FortAthenaBTService_AIEvaluator
         properties {
          interval=0.5
          nodeName=Evaluating... Bunker
          aIEvaluatorClass {defaultValue=/Game/Athena/AI/Phoebe/Evaluators/Phoebe_Evaluator_Bunker.Phoebe_Evaluator_Bunker_C}
         }
        }
        Evaluating... Take Cover {
         class=FortAthenaBTService_AIEvaluator
         properties {interval=0.5, nodeName=Evaluating... Take Cover, aIEvaluatorClass {defaultValue=/Script/FortniteAIServer.FortAthenaAIBotEvaluator_TakeCover}}
        }
        Evaluating...Observe Treat {
         class=FortAthenaBTService_AIEvaluator
         properties {
          bCallTickOnSearchStart=true
          interval=0.5
          nodeName=Evaluating...Observe Treat
          aIEvaluatorClass {defaultValue=/Game/Athena/AI/Phoebe/Evaluators/Phoebe_Evaluator_Observe.Phoebe_Evaluator_Observe_C}
         }
        }
       ]
      }
      Wait {class=BTTask_Wait, decorators [Loop {class=BTDecorator_Loop, properties {bInfiniteLoop=false}}], properties {waitTime {defaultValue=2}}}
     ]
     decorators [BTDecorator_HasEverSkydivedFromBusAndLanded_C_0 {
       class=BTDecorator_HasEverSkydivedFromBusAndLanded_C
       properties {
        bCheckConditionOnlyBlackBoardChanges=false
        bInverseCondition=false
        bShowPropertyDetails=true
        customDescription=""
        flowAbortMode=None
        nodeName=""
       }
      }]
     properties {
      nodeName=On Ground
      children [
       {
        childComposite=null
        childTask=&self:FortAthenaBTTask_ShootTrap_0
        decorators [&self:BTDecorator_Blackboard_3]
       }
       {
        childComposite=null
        childTask=&self:FortAthenaBTTask_BotMoveTo_4
        decorators [&self:BTDecorator_Blackboard_20]
       }
       {
        childComposite=null
        childTask=&self:FortAthenaBTTask_BotMoveTo_2
        decorators [&self:BTDecorator_Blackboard_16]
       }
       {
        childComposite=&self:BTComposite_Selector_5
        childTask=null
        decorators [&self:BTDecorator_Blackboard_18]
       }
       {
        childComposite=null
        childTask=&self:BTTask_Wait_3
        decorators [&self:BTDecorator_Loop_2]
       }
      ]
      services [
       &self:FortAthenaBTService_AIEvaluator_13
       &self:FortAthenaBTService_AIEvaluator_14
       &self:FortAthenaBTService_AIEvaluator_11
       &self:FortAthenaBTService_AIEvaluator_43
       &self:FortAthenaBTService_AIEvaluator_47
       &self:FortAthenaBTService_AIEvaluator_4
       &self:FortAthenaBTService_PickUpLoot_0
       &self:FortAthenaBTService_Crouch_1
       &self:FortAthenaBTService_Slide_0
       &self:FortAthenaBTService_Jump_1
       &self:FortAthenaBTService_Clamber_0
       &self:FortAthenaBTService_Sprinting_1
       &self:FortAthenaBTService_BuildConstruction_0
       &self:FortAthenaBTService_Escape_0
       &self:FortAthenaBTService_AIEvaluator_49
       &self:FortAthenaBTService_AIEvaluator_58
       &self:FortAthenaBTService_AIEvaluator_9
       &self:FortAthenaBTService_AIEvaluator_48
       &self:FortAthenaBTService_AIEvaluator_56
       &self:FortAthenaBTService_AIEvaluator_35
       &self:FortAthenaBTService_AIEvaluator_44
       &self:FortAthenaBTService_AIEvaluator_53
      ]
     }
     services [
      Evaluate...Loot {
       class=FortAthenaBTService_AIEvaluator
       properties {
        bCallTickOnSearchStart=true
        interval=0.5
        nodeName=Evaluate...Loot
        aIEvaluatorClass {defaultValue=/Script/FortniteAI.FortAthenaAIBotEvaluator_Loot}
       }
      }
      Evaluate...Sprinting {
       class=FortAthenaBTService_AIEvaluator
       properties {
        interval=0.2
        nodeName=Evaluate...Sprinting
        aIEvaluatorClass {defaultValue=/Game/Athena/AI/Phoebe/Evaluators/Phoebe_Evaluator_Sprinting.Phoebe_Evaluator_Sprinting_C}
       }
      }
      Evaluate...Revive {
       class=FortAthenaBTService_AIEvaluator
       properties {interval=0.5, nodeName=Evaluate...Revive, aIEvaluatorClass {defaultValue=/Script/FortniteAIServer.FortAthenaAIBotEvaluator_Revive}}
      }
      Evaluate...AvoidThreat {
       class=FortAthenaBTService_AIEvaluator
       properties {
        interval=0.5
        nodeName=Evaluate...AvoidThreat
        aIEvaluatorClass {defaultValue=/Game/Athena/AI/Phoebe/Evaluators/Phoebe_Evaluator_AvoidThreat.Phoebe_Evaluator_AvoidThreat_C}
       }
      }
      Evaluate...StealWall {
       class=FortAthenaBTService_AIEvaluator
       properties {interval=10, nodeName=Evaluate...StealWall, aIEvaluatorClass {defaultValue=/Script/FortniteAIServer.FortAthenaAIBotEvaluator_StealWall}}
      }
      Evaluate...Escape {
       class=FortAthenaBTService_AIEvaluator
       properties {
        interval=0.5
        nodeName=Evaluate...Escape
        aIEvaluatorClass {defaultValue=/Game/Athena/AI/Phoebe/Evaluators/Phoebe_Evaluator_Escape.Phoebe_Evaluator_Escape_C}
       }
      }
      Executing...Pick Up Loot {
       class=FortAthenaBTService_PickUpLoot
       properties {
        bCallTickOnSearchStart=false
        bRestartTimerOnEachActivation=false
        executionStatusName=AIEvaluator_Loot_ExecutionStatus
        interactionContextInfoName=AIEvaluator_Loot_InteractionContextInfo
        interactionExecutionStatusName=AIEvaluator_Loot_InteractionExecutionStatus
        interval=0.5
        lootObjectKeyName=AIEvaluator_Loot_Object
        movementStateKeyName=AIEvaluator_Loot_MovementState
        nodeName=Executing...Pick Up Loot
        randomDeviation=0.1
       }
      }
      Executing...Crouch {class=FortAthenaBTService_Crouch}
      Executing...Slide {
       class=FortAthenaBTService_Slide
       properties {
        bCallTickOnSearchStart=false
        bRestartTimerOnEachActivation=false
        interval=0.5
        nodeName=Executing...Slide
        randomDeviation=0.1
        slideExecutionStatusName=AIEvaluator_Slide_ExecutionStatus
       }
      }
      Executing...Jump {class=FortAthenaBTService_Jump, properties {crouchExecutionStatusName=AIEvaluator_Crouch_ExecutionStatus}}
      Executing... Clamber {
       class=FortAthenaBTService_Clamber
       properties {
        bCallTickOnSearchStart=false
        bRestartTimerOnEachActivation=false
        clamberAbilityStatusName=AIEvaluator_Clamber_AbilityStatus
        clamberAttemptRetryMaxCount=2
        clamberDestinationLocationName=AIEvaluator_Clamber_Destination
        clamberExecutionStatusName=AIEvaluator_Clamber_ExecutionStatus
        clamberOriginLocationName=AIEvaluator_Clamber_Origin
        clamberRetryDelay=0.15
        crouchExecutionStatusName=AIEvaluator_Crouch_ExecutionStatus
        firstJumpClamberMaxStartDelay=2
        firstJumpRetryDelay=0.1
        firstJumpRetryMaxCount=5
        interval=0.5
        jumpExecutionStatusName=AIEvaluator_Jump_ExecutionStatus
        nodeName=Executing... Clamber
        randomDeviation=0.1
       }
      }
      Executing...Sprinting {class=FortAthenaBTService_Sprinting}
      Executing...Build Construction {
       class=FortAthenaBTService_BuildConstruction
       properties {
        bCallTickOnSearchStart=false
        bRestartTimerOnEachActivation=false
        interval=0.5
        nodeName=Executing...Build Construction
        randomDeviation=0.1
        stealWallBuildGridCoordName=AIEvaluator_StealWallBuilding_GridCoord
        stealWallBuildName=AIEvaluator_StealWallBuilding_ExecutionStatus
        stealWallBuildTypeName=AIEvaluator_StealWallBuilding_Type
       }
      }
      Excuting...Escape {
       class=FortAthenaBTService_Escape
       properties {
        bCallTickOnSearchStart=false
        bRestartTimerOnEachActivation=false
        escapeFromStormKeyName=AIEvaluator_Storm_ExecutionStatus
        escapeKeyName=AIEvaluator_Escape_ExecutionStatus
        interval=0.5
        nodeName=Excuting...Escape
        randomDeviation=0.1
       }
      }
      Evaluating...Danger {
       class=FortAthenaBTService_AIEvaluator
       properties {
        interval=1000000000
        nodeName=Evaluating...Danger
        randomDeviation=0
        aIEvaluatorClass {defaultValue=/Game/Athena/AI/Phoebe/Evaluators/Phoebe_Evaluator_DangerDetection.Phoebe_Evaluator_DangerDetection_C}
       }
      }
      Evaluating...Trap On Path {
       class=FortAthenaBTService_AIEvaluator
       properties {
        interval=0.3
        nodeName=Evaluating...Trap On Path
        aIEvaluatorClass {defaultValue=/Script/FortniteAIServer.FortAthenaAIBotEvaluator_TrapOnPathDetected}
       }
      }
      Evaluate...Sand Tunnel {
       class=FortAthenaBTService_AIEvaluator
       properties {
        interval=100000000
        nodeName=Evaluate...Sand Tunnel
        aIEvaluatorClass {defaultValue=/Game/Athena/AI/Phoebe/Evaluators/Phoebe_Evaluator_SandTunnel.Phoebe_Evaluator_SandTunnel_C}
       }
      }
      Evaluating...Ground {
       class=FortAthenaBTService_AIEvaluator
       properties {interval=1, nodeName=Evaluating...Ground, aIEvaluatorClass {defaultValue=/Script/FortniteAIServer.FortAthenaAIBotEvaluator_Ground}}
      }
      Evaluating...Retreat {
       class=FortAthenaBTService_AIEvaluator
       properties {interval=0.5, nodeName=Evaluating...Retreat, aIEvaluatorClass {defaultValue=/Script/FortniteAIServer.FortAthenaAIBotEvaluator_Retreat}}
      }
      "Evaluating... Slot: Traversal" {
       class=FortAthenaBTService_AIEvaluator
       properties {
        interval=100000000
        nodeName=Evaluating... Slot: Traversal
        aIEvaluatorClass {defaultValue=null}
        injectionTag {tagName=Athena.AI.BTInjection.Phoebe.Traversal}
       }
      }
      "Evaluating... Slot: ChromeBlob" {
       class=FortAthenaBTService_AIEvaluator
       properties {
        interval=100000000
        nodeName=Evaluating... Slot: ChromeBlob
        aIEvaluatorClass {defaultValue=null}
        injectionTag {tagName=Athena.AI.BTInjection.Phoebe.Blob}
       }
      }
      "Evaluating... Slot: Grind" {
       class=FortAthenaBTService_AIEvaluator
       properties {
        interval=100000000
        nodeName=Evaluating... Slot: Grind
        aIEvaluatorClass {defaultValue=null}
        injectionTag {tagName=Athena.AI.BTInjection.Phoebe.Grind}
       }
      }
     ]
    }
   ]
   decorators []
   properties {
    nodeName=Alive
    children [
     {
      childComposite=null
      childTask=&self:FortAthenaBTTask_Zipline_0
      decorators [&self:BTDecorator_Blackboard_39]
     }
     {
      childComposite=null
      childTask=&self:FortAthenaBTTask_Dive_0
      decorators [&self:BTDecorator_Blackboard_21]
     }
     {
      childComposite=null
      childTask=&self:FortAthenaBTTask_Glide_0
      decorators [&self:BTDecorator_Blackboard_22]
     }
     {
      childComposite=&self:BTComposite_Selector_20
      childTask=null
      decorators [&self:BTDecorator_HasEverSkydivedFromBusAndLanded_C_0]
     }
    ]
    services [&self:FortAthenaBTService_AIEvaluator_1]
   }
   services [Evaluating...Free Fall {
     class=FortAthenaBTService_AIEvaluator
     properties {
      bCallTickOnSearchStart=true
      interval=0.5
      nodeName=Evaluating...Free Fall
      aIEvaluatorClass {defaultValue=/Game/Athena/AI/Phoebe/Evaluators/Phoebe_Evaluator_FreeFall.Phoebe_Evaluator_FreeFall_C}
     }
    }]
  }
 ]
 decorators [In Gameplay? {
   class=BTDecorator_Blackboard
   properties {
    arithmeticOperation=GreaterOrEqual
    flowAbortMode=None
    intValue=3
    nodeName=In Gameplay?
    stringValue=EAthenaGamePhase::Aircraft
    blackboardKey {selectedKeyName=AIEvaluator_Global_GamePhase}
   }
  }]
 properties {
  nodeName=Gameplay
  children [
   {
    childComposite=null
    childTask=&self:BTTask_Wait_7
    decorators [
     &self:BTDecorator_Blackboard_15
     &self:BTDecorator_Loop_3
    ]
   }
   {
    childComposite=null
    childTask=&self:FortAthenaBTTask_SteerMovement_0
    decorators [&self:BTDecorator_Blackboard_12]
   }
   {
    childComposite=null
    childTask=&self:BTTask_RunBehavior_5
    decorators [&self:BTDecorator_Blackboard_23]
   }
   {childComposite=&self:BTComposite_Selector_10, childTask=null, decorators []}
  ]
  services [
   &self:FortAthenaBTService_UpdateTarget_4
   &self:FortAthenaBTService_AIEvaluator_65
   &self:FortAthenaBTService_AIEvaluator_66
  ]
 }
 services [
  Evaluating...Targets {class=FortAthenaBTService_UpdateTarget, properties {interval=0.1}}
  Evaluating...Escape Storm {
   class=FortAthenaBTService_AIEvaluator
   properties {
    interval=1
    nodeName=Evaluating...Escape Storm
    randomDeviation=0.5
    aIEvaluatorClass {defaultValue=/Game/Athena/AI/Phoebe/Evaluators/Phoebe_Evaluator_Storm.Phoebe_Evaluator_Storm_C}
   }
  }
  Evaluating...Character Launched {
   class=FortAthenaBTService_AIEvaluator
   properties {
    interval=0.5
    nodeName=Evaluating...Character Launched
    aIEvaluatorClass {defaultValue=/Script/FortniteAIServer.FortAthenaAIBotEvaluator_CharacterLaunched}
   }
  }
 ]
}
